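# Niagara effect for the Majesty backpack cosmetic's trigger: a short glowing burst of drifting sparks around the pack bone.
# GBuffer_Particle spawns GPU sprites snapped onto the depth buffer with GBuffer color sampling, culled by validity and snap distance.
# Particles_for_Medium adds sparser, non-depth-projected sprites with similar curl-noise motion and white or orange tints.
# Mesh_Glow flashes one orange, growing mesh that follows the bone, with opacity scaled by time of day.
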NiagaraSystem /BRCosmetics/Characters/Player/Male/Medium/Bodies/M_MED_Majesty/FX/NS_Majesty_Backpack_Trigger
dependencies {
 usedDataInterfaces [&NiagaraDataInterfaceMeshRendererInfo, &NiagaraDataInterfaceGBuffer, &NiagaraDataInterfaceCurve]
 usedDynamicInputs []
 usedModules [&SystemState]
}
emitters {
 GBuffer_Particle {
  simTarget=GPUComputeSim
  emitterUpdateScript [
   &EmitterState(Life Cycle Mode=::NewEnumerator1, Loop Behavior=::NewEnumerator1, Loop Delay=0.5, Loop Duration=0.5, UseLoopDelay=false)
   &SpawnRate(SpawnRate=700, Use Spawn Probability=false)
  ]
  particleSpawnScript [
   /SetVariables() {isSetParameters=true}
   &InitializeParticle(
    Lifetime=&RandomRangeFloat(Minimum=0.3, Maximum=0.5)
    Sprite Rotation Mode=::NewEnumerator2
    Sprite Size Max=<1.5 1.5>
    Sprite Size Min=<1 1>
    Sprite Size Mode=::NewEnumerator2
    UsePositionOffset=false
    Write Material Random=true
   ) {module=InitializeParticle001}
   &SkeletalMeshLocation(
    Bone Sampling Mode=::NewEnumerator4
    MaskSpawnGroup=false
    Sampled Position Offset=<3 -4 0>
    Skeletal Mesh=DataInterface() {propertyValues {
      DefaultMesh=null
      ExcludeBoneName=root
      OptionalComponentClass=null
      PreviewMesh=/BRCosmetics/Accessories/FORT_Backpacks/Backpack_M_MED_Majesty/Meshes/M_MED_Majesty_Pack
      SkinningMode=SkinOnTheFly
      SoftSourceActor=null
      SourceMode=Default
      UvSetIndex=0
      WholeMeshLOD=-1
      bExcludeBone=false
      bReadDeformedGeometry=true
      bRequireCurrentFrameData=true
      ComponentTags []
      FilteredBones [pack]
      FilteredSockets []
      MeshUserParameter {parameter {name=None, typeDefHandle {registeredTypeIndex=1}, varData []}}
      SamplingRegions []
     }}
   )
   &BoxLocation(Box Size=<20 5 20>, MaskSpawnGroup=false, Module.Override Seed=false, Rotate Box=false, Surface Only=false)
   /SetVariables(Particles.InitialPos:NiagaraPosition=$Particles.Position) {isSetParameters=true}
   /Niagara/Modules/Math/PlaceParticlesOnDepthBuffer_GPU(Max Z Depth=1000, Module.Use Max Z Depth=true)
   /Game/Effects/Niagara/Modules/Spawn/FN_Spawn_GBuffer_Color()
   &KillParticles(Kill Particles=&SetBoolByFloatComparison(
     A=&VectorLength(Vector=&Subtract_Vector(A=$Particles.InitialPos, B=Linked() {linkedVariable {name=Particles.Position, type=NiagaraPosition}}))
     B=100
    ))
   &KillParticles(Kill Particles=&Boolean_NotOperation(Bool=$Output.PlaceParticlesOnDepthBuffer_GPU.ValidSample)) {module=KillParticles001}
   &AddVelocity(Velocity=<-40 -20 40>)
   &CameraOffset(Camera Offset Amount=10)
  ]
  particleUpdateScript [
   &ParticleState()
   &InheritSourceMovement(ApplyForce=false, ApplyPosition=true, ApplyPositionFalloff=true, ApplyPositionFalloffCurve=false)
   &CurlNoiseForce(
    Curl Noise Cone Mask Angle=70
    Mask Curl Noise=true
    Noise Frequency=80
    Noise Strength=500
    Pan Noise=true
    Pan Noise Field=<0 1 0>
    Random Seed=0
   )
   &ScaleColor(
    Scale Alpha=&FloatFromCurve(FloatCurve=DataInterface() {propertyValues {Curve {keys [
         <at -0.0014 0.0123>
         <at 0.0918 0.599>
         {arriveTangent=-0.5194, interpMode=RCIM_Cubic, leaveTangent=-0.5194, tangentMode=RCTM_User, time=1, value=0}
        ]}}})
    Scale RGB=&MakeVectorFromLinearColorRGB(Color=&MakeCustomLinearColorFromBool(Bool=&RandomBool(), True Color=<rgba 15 15 15 1>, False Color=<rgba 30 23.9823 15.3125 1>))
    ScaleA=true
    ScaleRGB=true
   )
   &GravityForce(Gravity=<0 0 -5>)
   &SolveForcesAndVelocity(Clamp Velocity=false, Limit Acceleration=false)
   /SetVariables(Particles.GetAge:NiagaraFloat=$Particles.NormalizedAge) {isSetParameters=true}
   &ScaleSpriteSize(Scale Factor=&Vector2DFromFloat(Value=&FloatFromCurve(FloatCurve=DataInterface() {propertyValues {Curve {keys [
          {interpMode=RCIM_Cubic, tangentMode=RCTM_User, time=-0.0044, value=0.1773}
          {arriveTangent=-0.1766, interpMode=RCIM_Cubic, leaveTangent=-0.1766, tangentMode=RCTM_User, time=0.1174, value=0.9496}
          {arriveTangent=-1.076, interpMode=RCIM_Cubic, leaveTangent=-1.076, tangentMode=RCTM_User, time=0.1639, value=0.7081}
          <at 1 0>
         ]}}})))
   &ScaleColor(
    Scale Alpha=&Multiply_Float(
     A=&FNTimeOfDayScalar(Night Scale=1, Morning Scale=1, Evening Scale=1)
     B=&FloatFromCurve(FloatCurve=DataInterface() {propertyValues {Curve {keys [<at -0.0014 0.0123>,<at 0.0918 0.599>,<at 1 0>]}}})
    )
    ScaleA=true
    ScaleRGB=true
   ) {module=ScaleColor001}
  ]
  rendererClasses [&NiagaraSpriteRendererProperties]
  renderers [{rendererClass=&NiagaraSpriteRendererProperties}]
 }
 Mesh_Glow {
  emitterUpdateScript [
   &EmitterState(Life Cycle Mode=::NewEnumerator1, Loop Behavior=::NewEnumerator1, Loop Delay=0.6, Loop Duration=1, UseLoopDelay=false)
   &SpawnBurst_Instantaneous(Spawn Count=1, Use Spawn Probability=false)
  ]
  particleSpawnScript [
   &InitializeParticle(
    Color=<rgba 25 5.0395 0 1>
    Lifetime=0.4
    Mesh Scale=<1.1 1.5 1.1>
    Mesh Scale Mode=::NewEnumerator4
    Sprite Size Mode=::NewEnumerator3
    Uniform Sprite Size=30
    UsePositionOffset=false
    Write Material Random=true
   )
   &CameraOffset(Camera Offset Amount=1)
  ]
  particleUpdateScript [
   &ParticleState()
   &SkeletalMeshLocation(
    Bone Sampling Mode=::NewEnumerator4
    MaskSpawnGroup=false
    Sampled Position Offset=<0 -3 0>
    Skeletal Mesh=DataInterface() {propertyValues {
      DefaultMesh=null
      ExcludeBoneName=root
      OptionalComponentClass=null
      PreviewMesh=&M_MED_Majesty_Pack
      SkinningMode=SkinOnTheFly
      SoftSourceActor=null
      SourceMode=Default
      UvSetIndex=0
      WholeMeshLOD=-1
      bExcludeBone=false
      bReadDeformedGeometry=true
      bRequireCurrentFrameData=true
      ComponentTags []
      FilteredBones [pack]
      FilteredSockets []
      MeshUserParameter {parameter {name=None, typeDefHandle {registeredTypeIndex=1}, varData []}}
      SamplingRegions []
     }}
   )
   &ScaleColor(
    Scale Alpha=&Multiply_Float(
     A=&FloatFromCurve(FloatCurve=DataInterface() {propertyValues {Curve {keys [
          {interpMode=RCIM_Cubic, tangentMode=RCTM_User, time=0, value=0}
          {arriveTangent=-3.9599, interpMode=RCIM_Cubic, leaveTangent=-3.9599, tangentMode=RCTM_User, time=0.1, value=1.05}
          {arriveTangent=-0.4177, interpMode=RCIM_Cubic, leaveTangent=-0.4177, tangentMode=RCTM_User, time=1, value=0}
         ]}}})
     B=&FNTimeOfDayScalar(Night Scale=0.5, Morning Scale=0.2, Evening Scale=0.5)
    )
    Scale RGB=&Multiply_Vector(
     A=<1 1 1>
     B=&VectorFromFloat(Value=&FloatFromCurve(FloatCurve=DataInterface() {propertyValues {Curve {keys [
           {interpMode=RCIM_Cubic, tangentMode=RCTM_User, time=0, value=0}
           {arriveTangent=1.1953, interpMode=RCIM_Cubic, leaveTangent=1.1953, tangentMode=RCTM_User, time=0.1, value=1.05}
           {arriveTangent=-4.7239, interpMode=RCIM_Cubic, leaveTangent=-4.7239, tangentMode=RCTM_User, time=1, value=0}
          ]}}}))
    )
    ScaleA=true
    ScaleRGB=true
   )
   &ScaleMeshSize(Scale Factor=&Multiply_Vector(
     A=&Multiply_Vector(
      A=<1 1 1>
      B=&VectorFromFloat(Value=&FloatFromCurve(FloatCurve=DataInterface() {propertyValues {Curve {keys [{interpMode=RCIM_Cubic, tangentMode=RCTM_User, time=0, value=1}, {interpMode=RCIM_Cubic, tangentMode=RCTM_User, time=1, value=1.5}]}}}))
     )
     B=$Engine.Owner.Scale
    ))
   &SolveForcesAndVelocity(Clamp Velocity=false, Limit Acceleration=false)
  ]
  rendererClasses [&NiagaraMeshRendererProperties]
  renderers [{rendererClass=&NiagaraMeshRendererProperties, rendererIndex=0}]
 }
 Particles_for_Medium {
  emitterUpdateScript [
   &EmitterState(Life Cycle Mode=::NewEnumerator1, Loop Behavior=::NewEnumerator1, Loop Delay=0.5, Loop Duration=0.4, UseLoopDelay=false)
   &SpawnRate(SpawnRate=40, Use Spawn Probability=false)
  ]
  particleSpawnScript [
   &InitializeParticle(
    Lifetime=&RandomRangeFloat(Minimum=0.3, Maximum=0.5)
    Sprite Rotation Mode=::NewEnumerator2
    Sprite Size Max=<2 2>
    Sprite Size Min=<1 1>
    Sprite Size Mode=::NewEnumerator2
    UsePositionOffset=false
    Write Material Random=true
   ) {module=InitializeParticle001}
   &SkeletalMeshLocation(
    Bone Sampling Mode=::NewEnumerator4
    MaskSpawnGroup=false
    Sampled Position Offset=<3 -4 0>
    Skeletal Mesh=DataInterface() {propertyValues {
      DefaultMesh=null
      ExcludeBoneName=root
      OptionalComponentClass=null
      PreviewMesh=&M_MED_Majesty_Pack
      SkinningMode=SkinOnTheFly
      SoftSourceActor=null
      SourceMode=Default
      UvSetIndex=0
      WholeMeshLOD=-1
      bExcludeBone=false
      bReadDeformedGeometry=true
      bRequireCurrentFrameData=true
      ComponentTags []
      FilteredBones [pack]
      FilteredSockets []
      MeshUserParameter {parameter {name=None, typeDefHandle {registeredTypeIndex=1}, varData []}}
      SamplingRegions []
     }}
   )
   &BoxLocation(Box Size=<20 5 20>, MaskSpawnGroup=false, Module.Override Seed=false, Rotate Box=false, Surface Only=false)
   &AddVelocity(Velocity=<-30 -20 30>)
   &CameraOffset(Camera Offset Amount=10)
  ]
  particleUpdateScript [
   &ParticleState()
   &InheritSourceMovement(ApplyForce=false, ApplyPosition=true, ApplyPositionFalloff=true, ApplyPositionFalloffCurve=false)
   &CurlNoiseForce(
    Curl Noise Cone Mask Angle=80
    Mask Curl Noise=true
    Noise Frequency=100
    Noise Strength=500
    Pan Noise=true
    Pan Noise Field=<0 1 0>
    Random Seed=0
   )
   &ScaleColor(
    Scale Alpha=&FloatFromCurve(FloatCurve=DataInterface() {propertyValues {Curve {keys [<at -0.0014 0.0123>,<at 0.0918 0.599>,<at 1 0>]}}})
    Scale RGB=&MakeVectorFromLinearColorRGB(Color=&MakeCustomLinearColorFromBool(Bool=&RandomBool(), True Color=<rgba 15 15 15 1>, False Color=<rgba 200 100.299 0 1>))
    ScaleA=true
    ScaleRGB=true
   )
   &GravityForce(Gravity=<0 0 -5>)
   &SolveForcesAndVelocity(Clamp Velocity=false, Limit Acceleration=false)
   &ScaleSpriteSize(Scale Factor=&Vector2DFromFloat(Value=&FloatFromCurve(FloatCurve=DataInterface() {propertyValues {Curve {keys [
          {interpMode=RCIM_Cubic, tangentMode=RCTM_User, time=-0.0044, value=0.1773}
          {arriveTangent=-0.1766, interpMode=RCIM_Cubic, leaveTangent=-0.1766, tangentMode=RCTM_User, time=0.1174, value=0.9496}
          {arriveTangent=-1.076, interpMode=RCIM_Cubic, leaveTangent=-1.076, tangentMode=RCTM_User, time=0.1639, value=0.7081}
          <at 0.8264 0.7215>
          <at 1 0>
         ]}}})))
   &ScaleColor(
    Scale Alpha=&Multiply_Float(
     A=1
     B=&FloatFromCurve(FloatCurve=DataInterface() {propertyValues {Curve {keys [<at -0.0014 0.0123>,<at 0.0918 0.599>,<at 1 0>]}}})
    )
    ScaleA=true
    ScaleRGB=true
   ) {module=ScaleColor001}
  ]
  rendererClasses [&NiagaraSpriteRendererProperties]
  renderers [{rendererClass=&NiagaraSpriteRendererProperties}]
 }
}
summary {
 systemName=NS_Majesty_Backpack_Trigger
 emitters [
  {emitterName=GBuffer_Particle, simTarget=GPUComputeSim, rendererClasses [&NiagaraSpriteRendererProperties]}
  {emitterName=Mesh_Glow, rendererClasses [&NiagaraMeshRendererProperties]}
  {emitterName=Particles_for_Medium, rendererClasses [&NiagaraSpriteRendererProperties]}
 ]
 userVariables [
  User.SK_MESH {
   type=NiagaraDataInterfaceSkeletalMesh
   defaultValue {
    struct=NiagaraExt_VariableValue_DataInterface
    value {dataInterface=&self:NiagaraDataInterfaceSkeletalMesh_0}
   }
  }
  User.Type_Transformation {type=NiagaraFloat, defaultValue=0.5, description=""}
 ]
}
