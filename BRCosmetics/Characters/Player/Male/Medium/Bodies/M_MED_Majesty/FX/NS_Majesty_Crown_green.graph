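# Niagara glow effect for the crown on the Majesty player cosmetic (male, medium body), green by default through User.CustomColor.
# A single sprite emitter bursts two persistent, camera-offset particles, using spawn group to pick each one's size and position offset.
# SkeletalMeshLocation runs every update to keep them on the User.SK_MESH skeleton.
# ScaleColor applies the custom color, with alpha modulated by time of day and front-end versus gameplay context.
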NiagaraSystem /BRCosmetics/Characters/Player/Male/Medium/Bodies/M_MED_Majesty/FX/NS_Majesty_Crown_green
dependencies {
 usedDataInterfaces [
  &NiagaraDataInterfaceArrayFloat
  &NiagaraDataInterfaceArrayFloat3
  &NiagaraDataInterfaceColorCurve
 ]
 usedDynamicInputs []
 usedModules [&SystemState]
}
emitters {Glow001 {
  emitterUpdateScript [
   &EmitterState(
    Inactive Response=::NewEnumerator1
    Life Cycle Mode=::NewEnumerator1
    Loop Behavior=::NewEnumerator1
    Loop Duration Mode=::NewEnumerator1
    UseLoopDelay=false
   )
   &SpawnBurst_Instantaneous(Spawn Count=2, Use Spawn Probability=false)
  ]
  particleSpawnScript [
   &FN_SetSpawnGroup(Number of Spawn Groups=2)
   &InitializeParticle(
    Sprite Rotation Mode=::NewEnumerator0
    Sprite Size Mode=::NewEnumerator3
    Uniform Sprite Size=&SelectFloatFromArray(Direct Array Index=$Particles.SpawnGroup, Float Selection Array=DataInterface() {propertyValues {FloatData [35,55]}})
    UsePositionOffset=false
    Write Material Random=true
   )
   &CameraOffset(Camera Offset Amount=2)
  ]
  particleUpdateScript [
   &SkeletalMeshLocation(
    Bone Sampling Mode=::NewEnumerator4
    MaskSpawnGroup=false
    Sampled Position Offset=&SelectVectorFromArray(Direct Array Index=$Particles.SpawnGroup, Vector Selection Array=DataInterface() {propertyValues {FloatData [<24 0 0>, <25 0 0>]}})
    Skeletal Mesh=$User.SK_MESH
   )
   &ParticleState(Kill Particles When Lifetime Has Elapsed=false, Let Infinitely Lived Particles Die When Emitter Deactivates=true)
   &ScaleColor(
    Scale Alpha=&Multiply_Float(
     A=&FNTimeOfDayScalar(Night Scale=0.3, Morning Scale=0.2, Day Scale=0.7, Evening Scale=0.2)
     B=&Multiply_Float(A=&FN_FrontEndFloatModifier(Gameplay=0.8, Front End=1), B=0.3)
    )
    Scale RGB=&MakeVectorFromLinearColorRGB(Color=$User.CustomColor)
    ScaleA=true
    ScaleRGB=true
   )
   &SolveForcesAndVelocity(Clamp Velocity=false, Limit Acceleration=false)
  ]
  rendererClasses [&NiagaraSpriteRendererProperties]
  renderers [{rendererClass=&NiagaraSpriteRendererProperties}]
 }}
summary {
 systemName=NS_Majesty_Crown_green
 emitters [{emitterName=Glow001, rendererClasses [&NiagaraSpriteRendererProperties]}]
 userVariables [
  User.bIsFrontEnd {type=NiagaraBool}
  User.bIsFrontEndPreview {type=NiagaraBool}
  User.CustomColor {type=LinearColor, description="", defaultValue {struct=LinearColor, value=<rgba 0.063 0.6445 0.0437 1>}}
  User.SK_MESH {
   type=NiagaraDataInterfaceSkeletalMesh
   defaultValue {
    struct=NiagaraExt_VariableValue_DataInterface
    value {dataInterface=&self:NiagaraDataInterfaceSkeletalMesh_0}
   }
  }
  User.Type_Transformation {type=NiagaraFloat, defaultValue=0, description=""}
 ]
}
