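# Niagara system of transformation sprites for the Majesty character; its only emitter, Real_time_Spaikies, is disabled.
# The GPU emitter spawns heavily only while System.bTrans and the front-end preview flag are true.
# Sprites from an ellipsoid are projected onto the depth buffer, take GBuffer color, and are killed on invalid or distant samples.
# They drift with curl noise and fade, tinted by a region mask that Type_Transformation moves and resizes.
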NiagaraSystem /BRCosmetics/Characters/Player/Male/Medium/Bodies/M_MED_Majesty/FX/NS_Majesty_Transform_sprites
dependencies {
 usedDataInterfaces [
  &NiagaraDataInterfaceMeshRendererInfo
  &NiagaraDataInterfaceSkeletalMesh
  &NiagaraDataInterfaceArrayFloat3
  &NiagaraDataInterfaceArrayFloat
  &NiagaraDataInterfaceGBuffer
  &NiagaraDataInterfaceCurve
  &NiagaraDataInterfaceDebugDraw
 ]
 usedDynamicInputs [&Abs_Float, &Subtract_Float]
 usedModules [
  &self:SystemSpawnScript.SystemScriptSource.SystemScriptGraph.NiagaraNodeAssignment_0.SetVariables_#1
  &SystemState
  &self:SystemSpawnScript.SystemScriptSource.SystemScriptGraph.NiagaraNodeAssignment_2.SetVariables_#2
  &self:SystemSpawnScript.SystemScriptSource.SystemScriptGraph.NiagaraNodeAssignment_1.SetVariables_#3
 ]
}
emitters {Real_time_Spaikies {
  bEnabled=false
  simTarget=GPUComputeSim
  emitterUpdateScript [
   &EmitterState(Life Cycle Mode=::NewEnumerator1, UseLoopDelay=false)
   &SpawnRate(
    SpawnRate=&MakeCustomFloatFromBool(
     Bool=$System.bTrans
     True Float=&MakeCustomFloatFromBool(Bool=$User.bisFrontEndPreview, True Float=10000, False Float=0)
     False Float=0
    )
    Use Spawn Probability=false
   )
  ]
  particleSpawnScript [
   &FN_SetSpawnGroup(Number of Spawn Groups=12) {enabled=false}
   /SetVariables() {isSetParameters=true}
   &InitializeParticle(
    Lifetime=&RandomRangeFloat(Minimum=0.3, Maximum=0.5)
    Sprite Rotation Mode=::NewEnumerator2
    Sprite Size Max=<1.5 1.5>
    Sprite Size Min=<1 1>
    Sprite Size Mode=::NewEnumerator2
    UsePositionOffset=false
    Write Material Random=true
   ) {module=InitializeParticle001}
   &SphereLocation(
    MaskSpawnGroup=false
    Non Uniform Scale=<1 1 2>
    Offset=&MakeVector(X=0, Y=0, Z=150)
    Override Seed=false
    Sphere Orientation Axis:Vector3f=<0 0 0>
    Sphere Radius=50
    Surface Only=false
    UseNonUniformScale=true
    UseOffset=true
    UseSphereOrientation=false
   ) {enabled=false}
   &SkeletalMeshLocation(
    Bone / Socket Index=$Particles.SpawnGroup
    Bone Sampling Mode=::NewEnumerator4
    MaskSpawnGroup=false
    Sampled Position Offset=&SelectVectorFromArray(
     Direct Array Index=$Particles.SpawnGroup
     Vector Selection Array=DataInterface() {propertyValues {FloatData [<0 0 0>, <0 0 0>, <0 0 0>, <0 0 0>, <0 0 0>, <0 0 0>, <0 0 0>, <0 0 0>, <0 0 0>, <0 0 0>, <0 0 0>]}}
    )
    Skeletal Mesh=DataInterface() {propertyValues {
      DefaultMesh=null
      ExcludeBoneName=root
      OptionalComponentClass=null
      PreviewMesh=&M_Med_Soldier_04_Preview
      SkinningMode=SkinOnTheFly
      SoftSourceActor=null
      SourceMode=Default
      UvSetIndex=0
      WholeMeshLOD=-1
      bExcludeBone=false
      bReadDeformedGeometry=true
      bRequireCurrentFrameData=true
      ComponentTags []
      FilteredBones [head,spine_01,spine_03,spine_05,upperarm_l,lowerarm_l,upperarm_r,lowerarm_r,thigh_l,thigh_r,calf_l,calf_r]
      FilteredSockets []
      MeshUserParameter {parameter {name=None, typeDefHandle {registeredTypeIndex=1}, varData []}}
      SamplingRegions []
     }}
   ) {enabled=false, module=SkeletalMeshLocation001}
   &SphereLocation(
    MaskSpawnGroup=false
    Offset=&MakeVector(X=0, Y=0, Z=0)
    Override Seed=false
    Sphere Orientation Axis:Vector3f=<0 0 0>
    Sphere Radius=&SelectFloatFromArray(
     Direct Array Index=$Particles.SpawnGroup
     Float Selection Array=DataInterface() {propertyValues {FloatData [20,20,20,20,10,10,10,10,10,10,10]}}
    )
    Surface Only=false
    UseNonUniformScale=true
    UseOffset=true
    UseSphereOrientation=false
   ) {enabled=false, module=SphereLocation001}
   &SphereLocation(
    MaskSpawnGroup=false
    Non Uniform Scale=<0.8 0.8 1.3>
    Offset=&MakeVector(X=0, Y=0, Z=120)
    Override Seed=false
    Sphere Orientation Axis:Vector3f=<0 0 0>
    Sphere Radius=50
    Surface Only=false
    UseNonUniformScale=true
    UseOffset=true
    UseSphereOrientation=false
   ) {module=SphereLocation002}
   /SetVariables(Particles.InitialPos:NiagaraPosition=$Particles.Position) {isSetParameters=true}
   /Niagara/Modules/Math/PlaceParticlesOnDepthBuffer_GPU(Max Z Depth=1000, Module.Use Max Z Depth=true)
   /Game/Effects/Niagara/Modules/Spawn/FN_Spawn_GBuffer_Color()
   &KillParticles(Kill Particles=&SetBoolByFloatComparison(
     A=&VectorLength(Vector=&Subtract_Vector(A=$Particles.InitialPos, B=Linked() {linkedVariable {name=Particles.Position, type=NiagaraPosition}}))
     B=100
    ))
   &KillParticles(Kill Particles=&Boolean_NotOperation(Bool=$Output.PlaceParticlesOnDepthBuffer_GPU.ValidSample)) {module=KillParticles001}
   &AddVelocity(Velocity=<20 0 20>)
   &CameraOffset(Camera Offset Amount=10)
  ]
  particleUpdateScript [
   &ParticleState()
   &Drag(Drag=2, Use Linear Drag=true, Use Rotational Drag=false)
   &InheritSourceMovement(Applied Position Falloff Distance=500, ApplyForce=false, ApplyPosition=true, ApplyPositionFalloff=true, ApplyPositionFalloffCurve=false)
   &CurlNoiseForce(Mask Curl Noise=true, Noise Frequency=80, Noise Strength=250, Pan Noise=false, Random Seed=6)
   &ScaleColor(
    Scale Alpha=&FloatFromCurve(FloatCurve=DataInterface() {propertyValues {Curve {keys [<at -0.0014 0.0123>,<at 0.0918 0.599>,<at 1 0>]}}})
    Scale RGB=&MakeVectorFromLinearColorRGB(Color=<rgba 10 10 10 1>)
    ScaleA=true
    ScaleRGB=true
   )
   &GravityForce(Gravity=<0 0 -5>)
   &SolveForcesAndVelocity(Clamp Velocity=false, Limit Acceleration=false)
   /SetVariables(Particles.GetAge:NiagaraFloat=$Particles.NormalizedAge) {isSetParameters=true}
   &ScaleSpriteSize(Scale Factor=&Vector2DFromFloat(Value=&FloatFromCurve(FloatCurve=DataInterface() {propertyValues {Curve {keys [
          {interpMode=RCIM_Cubic, tangentMode=RCTM_User, time=-0.0044, value=0.1773}
          {arriveTangent=-0.1766, interpMode=RCIM_Cubic, leaveTangent=-0.1766, tangentMode=RCTM_User, time=0.1174, value=0.9496}
          {arriveTangent=-1.076, interpMode=RCIM_Cubic, leaveTangent=-1.076, tangentMode=RCTM_User, time=0.1639, value=0.7081}
          <at 0.8264 0.7215>
          <at 1 0>
         ]}}})))
   &RegionMask(
    Inner Color:LinearColor=<rgba 4 1.8825 0 1>
    Offset:Vector3f=&Add_Vector(A=&Lerp_Vector(A=<0 0 50>, B=<0 0 130>, Alpha=$User.Type_Transformation), B=<0 0 0>)
    Offset Space:ENiagaraCoordinateSpace=::Local
    Outer Color:LinearColor=<rgba 1 1 1 1>
    Region Radius:NiagaraFloat=&Lerp_Float(A=50, B=200, Alpha=$User.Type_Transformation)
    Rotation:Vector3f=&TransformVector(Vector=<0 50 0>, Destination Space=::World)
    Rotation Angle Type:ENiagara_AngleInput=::NewEnumerator0
    UseOffset=true
    UseRotation=true
    Value Type:ENiagaraRegionMaskValue=::NewEnumerator2
   )
   &ScaleColor(
    Scale Alpha=&Multiply_Float(
     A=1
     B=&FloatFromCurve(FloatCurve=DataInterface() {propertyValues {Curve {keys [<at -0.0014 0.0123>,<at 0.0918 0.599>,<at 1 0>]}}})
    )
    Scale RGB=&MakeVectorFromLinearColorRGB(Color=$Particles.RegionMask.Color)
    ScaleA=true
    ScaleRGB=true
   ) {module=ScaleColor001}
  ]
  rendererClasses [&NiagaraSpriteRendererProperties]
  renderers [{rendererClass=&NiagaraSpriteRendererProperties}]
 }}
summary {
 systemName=NS_Majesty_Transform_sprites
 emitters [{bEnabled=false, emitterName=Real_time_Spaikies, simTarget=GPUComputeSim, rendererClasses [&NiagaraSpriteRendererProperties]}]
 userVariables [
  User.bisFrontEndPreview {type=NiagaraBool}
  User.Type_Transformation {type=NiagaraFloat, defaultValue=0, description=""}
 ]
}
