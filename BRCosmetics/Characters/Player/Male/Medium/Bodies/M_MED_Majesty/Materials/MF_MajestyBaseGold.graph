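# Material function that supplies the base gold surface for the Majesty character body materials.
# A static switch selects either gold color, specular, metallic and roughness parameters or diffuse and packed spec-mask textures sampled with TexCoords.
# Roughness is remapped between min/max parameters, with an optional mobile range at Low quality.
# It outputs assembled material attributes plus separate diffuse, normal, roughness, specular mask and auxiliary M mask channels.
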MaterialFunction /BRCosmetics/Characters/Player/Male/Medium/Bodies/M_MED_Majesty/Materials/MF_MajestyBaseGold
graph {nodes [
  Constant3Vector() {constant=<rgba 0.2086 0.2086 0.2086 1>}
  n1 = FunctionInput(Preview=TextureCoordinate()) {bUsePreviewValueAsDefault=true, inputName=TexCoords, inputType=FunctionInput_Vector2}
  n2 = StaticSwitchParameter(
   True=VectorParameter() {parameterName=Gold_Diffuse, defaultValue=<rgba 0.7231 0.3613 0.1413 1>}.RGB
   False=TextureSampleParameter2D(UVs=@n1) {group=Diffuse, parameterName=Diffuse_G, texture=/Engine/EngineMaterials/DefaultDiffuse}.RGB
  ) {parameterName=UseParameterValuesForSurface}
  n3 = ScalarParameter() {group=Roughness, parameterName=MobileRoughnessMin}
  n4 = ScalarParameter() {defaultValue=1, group=Roughness, parameterName=MobileRoughnessMax}
  n5 = ScalarParameter() {defaultValue=0.2086, parameterName=Roughness}
  n6 = TextureSampleParameter2D(UVs=@n1) {
   desc="R = SpecMask G = Metallic B = Roughness"
   group=Specular
   parameterName=SpecularMasks_G
   samplerType=SAMPLERTYPE_Masks
   texture=&T_Fortnite_Default_S
  }
  n7 = StaticSwitchParameter(True=@n5, False=@n6.B) {parameterName=UseParameterValuesForSurface}
  n8 = LinearInterpolate(
   A=ScalarParameter() {group=Roughness, parameterName=RoughnessMin}
   B=ScalarParameter() {defaultValue=1, group=Roughness, parameterName=RoughnessMax}
   Alpha=@n7
  )
  n9 = QualitySwitch(Default=@n8, Low=StaticSwitchParameter(True=LinearInterpolate(A=@n3, B=@n4, Alpha=@n7), False=@n8) {parameterName=OverrideRoughnessOnMobile})
  n10 = Constant()
  n11 = TextureSampleParameter2D(UVs=@n1) {group=Normal, parameterName=Normals_G, samplerType=SAMPLERTYPE_Normal, texture=/Engine/EngineMaterials/DefaultNormal}
  n12 = ScalarParameter() {defaultValue=1, parameterName=Specular}
  n13 = ScalarParameter() {defaultValue=1, parameterName=Metallic}
  FunctionOutput(None=MakeMaterialAttributes!(
    BaseColor=@n2
    Metallic=StaticSwitchParameter(True=@n13, False=@n6.G) {parameterName=UseParameterValuesForSurface}
    Specular=StaticSwitchParameter(True=@n12, False=@n6.R) {parameterName=UseParameterValuesForSurface}
    Roughness=@n9
    Anisotropy
    EmissiveColor=@n10
    Opacity=Constant()
    OpacityMask
    Normal=@n11.RGB
    Tangent
    WorldPositionOffset
    SubsurfaceColor=@n10
    ClearCoat=@n10
    ClearCoatRoughness=@n10
    AmbientOcclusion
    Refraction
    CustomizedUVs_0
    CustomizedUVs_1
    CustomizedUVs_2
    CustomizedUVs_3
    CustomizedUVs_4
    CustomizedUVs_5
    CustomizedUVs_6
    CustomizedUVs_7
    PixelDepthOffset
    ShadingModel=ShadingModel()
    Displacement
   ))
  FunctionOutput(None=@n11.RGB) {outputName=Normal Map, sortPriority=2}
  FunctionOutput(None=@n9) {outputName=Roughness, sortPriority=8}
  FunctionOutput(None=StaticSwitchParameter(
    True=MaterialFunctionCall(X=@n12, Y=@n13, Z=@n5) {materialFunction=&MakeFloat3}.Result
    False=@n6.RGB
   ) {parameterName=UseParameterValuesForSurface}) {outputName=SpecularMask, sortPriority=3}
  n14 = TextureSampleParameter2D(UVs=@n1) {
   desc="Masks R=AO G= Fresnel B = Tmission/Skin"
   group=Masks
   parameterName=M_G
   samplerType=SAMPLERTYPE_Masks
   texture=/Engine/EngineMaterials/DefaultDiffuse_TC_Masks
  }
  FunctionOutput(None=@n14.RGB) {outputName=M, sortPriority=4}
  FunctionOutput(None=@n2) {outputName=DIffuse, sortPriority=1}
 ]}
