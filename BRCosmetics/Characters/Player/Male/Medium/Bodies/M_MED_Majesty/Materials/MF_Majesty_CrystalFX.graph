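# Material function that adds a glowing crystal effect to the Majesty character body material.
# A pre-skinned height gradient, Gmap channel, Fresnel and parallax crack texture form a value indexing curve-atlas color ramps, switchable to a custom tint.
# Rim highlights finish standard and gold variants, which GoldMask blends into base color and emissive.
# The main output applies this through Mask; CrystalGold outputs the gold variant unmasked.
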MaterialFunction /BRCosmetics/Characters/Player/Male/Medium/Bodies/M_MED_Majesty/Materials/MF_Majesty_CrystalFX
graph {nodes [
  n1 = PreSkinnedPosition()
  n2 = FunctionInput(Preview=ShaderStageSwitch(PixelShader=VertexInterpolator(VS=@n1).PS, VertexShader=@n1)) {inputName=PreskinnedPositon}
  n3 = VectorParameter() {parameterName=BoundsGradientOffset, defaultValue=<rgba 0 0 -170 1>}
  n4 = PreSkinnedLocalBounds()
  n5 = ScalarParameter() {defaultValue=0.025, parameterName=BoundsGradientFalloff}
  n6 = ChannelMaskParameter(None=Divide(A=Add(A=@n2, B=@n3.RGB), B=Multiply(A=@n4.Max, B=@n5))) {maskChannel=Blue, parameterName=Z_Mask, defaultValue=<rgba 0 0 1 0>}
  n7 = ComponentMask(None=ChannelMaskParameter(None=FunctionInput() {inputName=Gmap}) {parameterName=GMapChannelMask}.R) {r=true}
  n8 = Power(Base=Fresnel(ExponentIn=ScalarParameter() {defaultValue=0.5, parameterName=FresnelExponent}, Normal=PixelNormalWS()))
  n9 = TextureCoordinate()
  n10 = AppendVector(
   A=ScalarParameter() {defaultValue=2, parameterName=CrystalPatternTiling_X}
   B=ScalarParameter() {defaultValue=2, parameterName=CrystalPatternTiling_Y}
  )
  n11 = TextureSample(UVs=BumpOffset(
    Coordinate=Multiply(A=@n9, B=@n10)
    Height=ScalarParameter() {defaultValue=1, parameterName=CrystalPatternHeight}
    HeightRatioInput=ScalarParameter() {defaultValue=-0.25, parameterName=CrystalPatternHeightRatio}
   ) {referencePlane=0}) {samplerType=SAMPLERTYPE_Masks, texture=/Game/Weapons/WeaponSkins/Wraps/Textures/IceBreaker/T_IceBreakerCracks}
  n12 = Clamp(
   None=Add(A=@n11.R, B=@n11.R)
   Min=ScalarParameter() {defaultValue=0.5, parameterName=CrystalPatternMinIntensity}
   Max=LinearInterpolate(
    A=ScalarParameter() {defaultValue=1, parameterName=CrystalPatternMaxIntensityBot}
    B=ScalarParameter() {defaultValue=0.9, parameterName=CrystalPatternMaxIntensity}
    Alpha=@n6.R
   )
  )
  n13 = Multiply(A=Add(A=Multiply(A=@n6.R, B=@n7), B=Clamp(None=@n8)), B=@n12)
  Add(A=@n13, B=@n12)
  n14 = GetMaterialAttributes(None=FunctionInput() {inputName=Material, inputType=FunctionInput_MaterialAttributes}) {attributeGetTypes [#1, #2]}
  ComponentMask(None=@n14.BaseColor) {r=true}
  n15 = CurveAtlasRowParameter(CurveTime=@n13) {
   atlas=&AthenaCurveAtlas01
   curve=/Game/Characters/Player/Female/Medium/Bodies/F_MED_Network/Textures/Curve_Network_Silver
   defaultValue=314
  }
  n16 = Multiply(
   A=Desaturation(
    None=VectorParameter() {group=CustomColor, parameterName=CustomColor, defaultValue=<rgba 0.9062 0.3982 0.8048 0>}.RGB
    Fraction=ScalarParameter() {defaultValue=0.5, group=CustomColor, parameterName=CustomColor_Desat}
   )
   B=LinearInterpolate(
    A=ScalarParameter() {group=CustomColor, parameterName=CustomColor_Min}
    B=ScalarParameter() {defaultValue=1, group=CustomColor, parameterName=CustomColor_Max}
    Alpha=@n15
   )
  )
  n17 = CurveAtlasRowParameter(CurveTime=@n13) {
   atlas=&AthenaCurveAtlas01
   curve=/Game/Characters/Player/Male/Medium/Bodies/M_MED_Majesty/Textures/MajestyEmissiveCurve
   defaultValue=288
   parameterName=EmissiveColorCurve
  }
  n18 = StaticSwitchParameter(True=@n16, False=@n17) {group=CustomColor, parameterName=Use_CustomColor}
  n19 = LinearInterpolate(
   A=ScalarParameter() {defaultValue=2, parameterName=Max_EmissiveStrength}
   B=ScalarParameter() {defaultValue=1.5, parameterName=Min_EmissiveStrength}
   Alpha=@n6.R
  )
  n20 = ScalarParameter() {defaultValue=-4, group=HoloFoil, parameterName=BL_Rim_OutlineC1}
  n21 = ScalarParameter() {defaultValue=6, group=HoloFoil, parameterName=BL_Rim_OutlineC2}
  n22 = ScalarParameter() {defaultValue=4, group=HoloFoil, parameterName=BL_RimHighlightExponent1}
  n23 = PixelNormalWS()
  n24 = Power(
   Base=OneMinus(None=Fresnel(ExponentIn=@n22, Normal=@n23) {baseReflectFraction=0})
   Exp=ScalarParameter() {defaultValue=4, group=HoloFoil, parameterName=BL_Rim_OutlineExponent2}
  )
  n25 = OneMinus(None=Saturate(None=LinearInterpolate(A=@n20, B=@n21, Alpha=@n24)))
  n26 = LinearInterpolate(A=Power(Base=Multiply(A=@n18, B=@n19)), B=Constant3Vector() {constant=<rgba 1 1 1 0>}, Alpha=@n25)
  n27 = CurveAtlasRowParameter(CurveTime=@n13) {
   atlas=&AthenaCurveAtlas01
   curve=/Game/Characters/Player/Male/Medium/Bodies/M_MED_Majesty/Textures/MajestyGoldEmissiveCurve
   defaultValue=291
   parameterName=EmissiveGoldColor
  }
  n28 = StaticSwitchParameter(True=@n16, False=@n27) {group=CustomColor, parameterName=Use_CustomColor}
  n29 = LinearInterpolate(
   A=ScalarParameter() {parameterName=Max_GoldEmissiveStrength}
   B=ScalarParameter() {defaultValue=1.5, parameterName=Min_GoldEmissiveStrength}
   Alpha=@n6.R
  )
  n30 = LinearInterpolate(A=Power(Base=Multiply(A=@n28, B=@n29)), B=Constant3Vector() {constant=<rgba 1 0.5175 0 0>}, Alpha=@n25)
  n31 = FunctionInput(Preview=Constant()) {inputName=GoldMask, inputType=FunctionInput_Scalar}
  n32 = SetMaterialAttributes!(
   MaterialAttributes=@n14.MaterialAttributes
   EmissiveColor=Add(A=@n14.EmissiveColor, B=LinearInterpolate(A=@n26, B=@n30, Alpha=@n31))
   BaseColor=LinearInterpolate(A=@n26, B=@n30, Alpha=@n31)
   Metallic=Constant()
   Roughness=Constant() {r=0.4}
   Specular=Constant() {r=1}
  ) {attributeSetTypes [
    #1
    #2
    #3
    #4
    #5
   ]}
  n33 = FunctionInput() {inputName=Mask, inputType=FunctionInput_Vector4}
  FunctionOutput(None=BlendMaterialAttributes(A=@n14.MaterialAttributes, B=@n32, Alpha=ChannelMaskParameter(None=@n33) {parameterName=MaskChannel}.R) {vertexAttributeBlendType=UseA})
  n34 = SetMaterialAttributes(
   MaterialAttributes=@n14.MaterialAttributes
   EmissiveColor=Add(A=@n14.EmissiveColor, B=@n30)
   BaseColor=@n30
   Metallic=Constant()
   Roughness=Constant() {r=0.4}
   Specular=Constant() {r=1}
  ) {attributeSetTypes [
    #1
    #2
    #3
    #4
    #5
   ]}
  FunctionOutput(None=@n34) {outputName=CrystalGold}
  ScalarParameter() {parameterName=BoundsGradientOffset}
 ]}
