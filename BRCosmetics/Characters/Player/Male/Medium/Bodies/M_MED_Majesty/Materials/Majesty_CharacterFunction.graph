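# Material function for the Majesty player character that builds base surface attributes, blending standard and "_G" texture sets by a GoldMask input.
# GoldMask lerps diffuse and packed specular masks, whose channels feed specular, metallic and roughness; normals use one shared texture.
# Roughness is remapped between min/max parameters, with an optional mobile range on low quality.
# It outputs material attributes plus separate diffuse, normal, specular mask, roughness and AO/fresnel/skin mask values.
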MaterialFunction /BRCosmetics/Characters/Player/Male/Medium/Bodies/M_MED_Majesty/Materials/Majesty_CharacterFunction
graph {nodes [
  n1 = FunctionInput(Preview=TextureCoordinate()) {bUsePreviewValueAsDefault=true, inputName=TexCoords, inputType=FunctionInput_Vector2}
  n2 = FunctionInput(Preview=Constant()) {bUsePreviewValueAsDefault=true, inputName=GoldMask, inputType=FunctionInput_Scalar}
  n3 = LinearInterpolate(
   A=TextureSampleParameter2D(UVs=@n1) {group=Diffuse, parameterName=Diffuse, texture=/Engine/EngineMaterials/DefaultDiffuse}.RGB
   B=TextureSampleParameter2D(UVs=@n1) {group=Diffuse, parameterName=Diffuse_G, texture=&DefaultDiffuse}.RGB
   Alpha=@n2
  )
  n4 = ScalarParameter() {group=Roughness, parameterName=MobileRoughnessMin}
  n5 = ScalarParameter() {defaultValue=1, group=Roughness, parameterName=MobileRoughnessMax}
  n6 = TextureSampleParameter2D(UVs=@n1) {
   desc="R = SpecMask G = Metallic B = Roughness"
   group=Specular
   parameterName=SpecularMasks
   samplerType=SAMPLERTYPE_Masks
   texture=&T_Fortnite_Default_S
  }
  n7 = TextureSampleParameter2D(UVs=@n1) {
   desc="R = SpecMask G = Metallic B = Roughness"
   group=Specular
   parameterName=SpecularMasks_G
   samplerType=SAMPLERTYPE_Masks
   texture=&T_Fortnite_Default_S
  }
  n8 = LinearInterpolate(A=@n6.RGBA, B=@n7.RGBA, Alpha=@n2)
  n9 = ChannelMaskParameter(None=@n8) {maskChannel=Blue, parameterName=B, defaultValue=<rgba 0 0 1 0>}
  n10 = LinearInterpolate(
   A=ScalarParameter() {group=Roughness, parameterName=RoughnessMin}
   B=ScalarParameter() {defaultValue=1, group=Roughness, parameterName=RoughnessMax}
   Alpha=@n9.R
  )
  n11 = QualitySwitch(
   Default=@n10
   Low=StaticSwitchParameter(True=LinearInterpolate(A=@n4, B=@n5, Alpha=@n9.R), False=@n10) {parameterName=OverrideRoughnessOnMobile}
  )
  n12 = Constant()
  n13 = TextureSampleParameter2D(UVs=@n1) {group=Normal, parameterName=Normals, samplerType=SAMPLERTYPE_Normal, texture=/Engine/EngineMaterials/DefaultNormal}
  FunctionOutput(None=MakeMaterialAttributes!(
    BaseColor=@n3
    Metallic=ChannelMaskParameter(None=@n8) {maskChannel=Green, parameterName=G, defaultValue=<rgba 0 1 0 0>}.R
    Specular=ChannelMaskParameter(None=@n8) {parameterName=R}.R
    Roughness=@n11
    Anisotropy
    EmissiveColor=@n12
    Opacity=Constant()
    OpacityMask
    Normal=@n13.RGB
    Tangent
    WorldPositionOffset
    SubsurfaceColor=@n12
    ClearCoat=@n12
    ClearCoatRoughness=@n12
    AmbientOcclusion
    Refraction
    CustomizedUVs_0
    CustomizedUVs_1
    CustomizedUVs_2
    CustomizedUVs_3
    CustomizedUVs_4
    CustomizedUVs_5
    CustomizedUVs_6
    CustomizedUVs_7
    PixelDepthOffset
    ShadingModel=ShadingModel()
    Displacement
   ))
  FunctionOutput(None=@n13.RGB) {outputName=Normal Map, sortPriority=2}
  FunctionOutput(None=@n11) {outputName=Roughness, sortPriority=8}
  FunctionOutput(None=ComponentMask(None=@n8) {b=true, g=true, r=true}) {outputName=SpecularMask, sortPriority=3}
  n14 = TextureSampleParameter2D(UVs=@n1) {
   desc="Masks R=AO G= Fresnel B = Tmission/Skin"
   group=Masks
   parameterName=M
   samplerType=SAMPLERTYPE_Masks
   texture=/Engine/EngineMaterials/DefaultDiffuse_TC_Masks
  }
  FunctionOutput(None=@n14.RGB) {outputName=M, sortPriority=4}
  FunctionOutput(None=@n3) {outputName=DIffuse, sortPriority=1}
 ]}
