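# Material function that builds the Majesty male medium body's character shading, exposing masks and helper values alongside the result, apparently for later FX.
# Majesty_CharacterFunction supplies material, normal and packed AO/GMap/Skin mask, feeding shared GMap color, brightness balance, emissive and skin functions in sequence.
# A normal-mapped Fresnel drives emissive, whose final color passes through a Power.
# Extra outputs include masks, Fresnel, distance blend and pre-skinned local position, vertex-interpolated for pixel shaders.
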MaterialFunction /BRCosmetics/Characters/Player/Male/Medium/Bodies/M_MED_Majesty/Materials/MF_Majesty_PreFX
graph {nodes [
  n1 = FunctionInput(Preview=TextureCoordinate()) {bUsePreviewValueAsDefault=true, inputName=TexCoords, inputType=FunctionInput_Vector2}
  n2 = MaterialFunctionCall!(GoldMask=FunctionInput(Preview=Constant()) {inputName=GoldMask, inputType=FunctionInput_Scalar}, TexCoords=@n1) {materialFunction=&/Majesty_CharacterFunction}
  n3 = MaterialFunctionCall(Float3=@n2.M) {desc="R = AO\r\nG = GMap\r\nB = Skin", materialFunction=&BreakOutFloat3Components}
  n4 = MaterialFunctionCall!(Material=@n2.Result, Diffuse=@n2.Result, GMap=@n3.G) {materialFunction=/Game/Packages/Fortress_SharedMaterials/Base_Material_Functions/BaseCharacterShaderFunctions/BaseFN_CharacterGMapColor}
  n5 = MaterialFunctionCall!(Material=@n4.Result) {materialFunction=/Game/Packages/Fortress_SharedMaterials/Base_Material_Functions/CharacterBrightnessandScale_Balance}
  n6 = Fresnel(Normal=Transform(None=@n2.Normal Map)) {baseReflectFraction=0, exponent=1}
  n7 = GetMaterialAttributes(None=MaterialFunctionCall!(
    Material=MaterialFunctionCall!(Material=@n5.Result, Diffuse=@n4.DIffuse, BalanceDistance=@n5.Result, Normals=@n2.Normal Map, AO=@n3.R, Fresnel=@n6) {materialFunction=/Game/Packages/Fortress_SharedMaterials/Base_Material_Functions/BaseCharacterShaderFunctions/BaseFN_CharacterEmissive}.Result
    AO=@n3.R
    Skin Mask=@n3.B
   ) {materialFunction=/Game/Packages/Fortress_SharedMaterials/Base_Material_Functions/BaseCharacterShaderFunctions/BaseFN_CharacterSkin}.Result) {attributeGetTypes [#1]}
  FunctionOutput(None=SetMaterialAttributes!(MaterialAttributes=@n7.MaterialAttributes, EmissiveColor=Power(Base=@n7.EmissiveColor)) {attributeSetTypes [#1]})
  FunctionOutput(None=@n3.R) {outputName=AO}
  FunctionOutput(None=@n3.G) {outputName=Gmap}
  FunctionOutput(None=@n3.B) {outputName=Skin}
  FunctionOutput(None=@n6) {outputName=Fresnel}
  FunctionOutput(None=@n5.Distance Blend) {outputName=Distance Blend}
  n8 = PreSkinnedPosition()
  FunctionOutput(None=ShaderStageSwitch(PixelShader=VertexInterpolator(VS=@n8).PS, VertexShader=@n8)) {outputName=Pre skinned Local Position}
  MaterialFunctionCall!(TexCoords=@n1) {materialFunction=/Game/Packages/Fortress_SharedMaterials/Base_Material_Functions/BaseCharacterShaderFunctions/BaseFN_CharacterFunction}
  Saturate(None=DotProduct(A=PixelNormalWS(), B=CameraVectorWS()))
  SetMaterialAttributes!(MaterialAttributes, BaseColor) {attributeSetTypes [#2]}
 ]}
