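# Material function that blends two material attribute layers on the Majesty outfit with an animated, glowing gold transition.
# A pre-skinned-position sphere mask, distorted by panning noise and staggered by vertex color, drives the blend and is also output as GoldMask.
# Sampling narrow bands around the GMap-adjusted mask creates emissive edges and gradients for a projected, panning dot pattern.
# Super-levelling parameters derive quantizable progress from a transform value and Blueprint-supplied start time.
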MaterialFunction /BRCosmetics/Characters/Player/Male/Medium/Bodies/M_MED_Majesty/Materials/MLB_MajestyGoldBlend
graph {nodes [
  n1 = PreSkinnedPosition()
  n2 = FunctionInput(Preview=ShaderStageSwitch(PixelShader=VertexInterpolator(VS=@n1).PS, VertexShader=@n1)) {bUsePreviewValueAsDefault=true, inputName=PreskinnedPos}
  n3 = VectorParameter() {group=SuperLevelling_Mask, parameterName=SL_Backpack_PreskinnedPositionOffset, defaultValue=<rgba 0 0 0 1>}
  Add(A=@n2, B=@n3.RGB)
  n4 = NamedRerouteUsage()
  n5 = Add(A=@n4, B=Multiply(A=Time(), B=VectorParameter() {parameterName=PatternPanSpeedAndDirection, defaultValue=<rgba 1 2 -2 1>}.RGB))
  n6 = ScalarParameter() {defaultValue=175, parameterName=Pattern_Tiling}
  n7 = Frac(None=Divide(A=@n5, B=@n6))
  n8 = ComponentMask(None=@n7) {g=true, r=true}
  n9 = ComponentMask(None=@n7) {b=true, g=true}
  n10 = PreSkinnedNormal()
  n11 = ShaderStageSwitch(PixelShader=VertexInterpolator(VS=@n10).PS, VertexShader=@n10)
  n12 = ComponentMask(None=@n11) {r=true}
  n13 = LinearInterpolate(
   A=LinearInterpolate(
    A=TextureSample(UVs=@n8) {texture=/Game/Athena/Environments/Nexus/Textures/T_Tile_Dots}.RGB
    B=TextureSample(UVs=@n9) {texture=&T_Tile_Dots}.RGB
    Alpha=Abs(None=@n12)
   )
   B=TextureSample(UVs=ComponentMask(None=@n7) {b=true, r=true}) {texture=&T_Tile_Dots}.RGB
   Alpha=Abs(None=ComponentMask(None=@n11) {g=true})
  )
  n14 = Subtract(A=Multiply(A=@n13) {constB=2})
  n15 = ScalarParameter() {defaultValue=1, parameterName=MaxOffset}
  n16 = MaterialFunctionCall(Vec3=Ceil(None=Divide(A=@n4, B=@n6))) {materialFunction=&MF_Hash}
  n17 = ComponentMask(None=@n16.Vec3) {g=true, r=true}
  Add(A=@n14, B=Multiply(A=@n15, B=@n17))
  n18 = TextureObjectParameter() {
   group=SuperLevelling_Mask
   parameterName=SL_TransitionNoise
   texture=/Game/Characters/Player/Male/Medium/Bodies/M_MED_Viper/Textures/Glitch/T_SoftNoise
  }
  n19 = Time()
  n20 = VectorParameter() {group=Cinematic, parameterName=CINE_DistortionNoise_PanDirectionAndSpeed, defaultValue=<rgba 4 4 2 1>}
  n21 = PreSkinnedNormal()
  n22 = MaterialFunctionCall(
   TextureObject=@n18
   TextureSize=VectorParameter() {group=SuperLevelling_Mask, parameterName=SL_DistortionNoise_Tiling, defaultValue=<rgba 32 32 32 1>}.RGB
   WorldPosition=Add(A=@n2, B=Multiply(A=@n19, B=@n20.RGB))
   World Space Normal=FunctionInput(Preview=VertexInterpolator(VS=@n21).PS) {bUsePreviewValueAsDefault=true, inputName=Local Normal}
  ) {materialFunction=/Game/Weapons/WeaponSkins/Wraps/MaterialFunctions/WorldAlignedTexture}
  n23 = NamedRerouteUsage()
  n24 = OneMinus(None=@n23)
  n25 = Multiply(
   A=Subtract(A=Multiply(A=@n22.XYZ Texture) {constB=2})
   B=LinearInterpolate(B=ScalarParameter() {defaultValue=2, group=Cinematic, parameterName=CINE_DistortionStrength}, Alpha=Ceil(None=@n24))
  )
  Add(A=@n25, B=@n5)
  n26 = NamedRerouteUsage()
  ChannelMaskParameter(None=Divide(A=@n26, B=MaterialFunctionCall() {materialFunction=&ObjectLocalBounds}.Local Bounds Max)) {maskChannel=Blue, parameterName=zMASK, defaultValue=<rgba 0 0 1 0>}
  Constant()
  Constant() {r=2}
  Constant() {r=1}
  Constant() {r=1}
  Constant()
  n27 = GetMaterialAttributes(None=FunctionInput() {inputName=Material_01, inputType=FunctionInput_MaterialAttributes}) {attributeGetTypes [#1]}
  n28 = Multiply(A=NamedRerouteUsage(), B=VectorParameter() {parameterName=EmissiveEdgeColor, defaultValue=<rgba 1 0.5903 0 1>}.RGB)
  n29 = NamedRerouteUsage()
  n30 = VectorParameter() {parameterName=SphereMaskLocation_01, defaultValue=<rgba 0 0 166.7103 1>}
  n31 = ScalarParameter() {defaultValue=32, parameterName=SphereMaskRadius}
  n32 = SphereMask(A=@n30.RGB, B=@n26, Radius=@n31, Hardness=Constant())
  n33 = Multiply(
   A=OneMinus(None=@n29)
   B=SphereMask(A=OneMinus(None=@n32), B=@n29, Radius=ScalarParameter() {defaultValue=0.05, parameterName=EdgeGlowRadius}, Hardness=Constant())
  )
  n34 = Power(Base=Multiply(A=Multiply(A=@n28, B=@n33), B=Constant() {r=10}))
  n35 = Add(A=@n27.EmissiveColor, B=@n34)
  n36 = GetMaterialAttributes(None=FunctionInput() {inputName=Material_02, inputType=FunctionInput_MaterialAttributes}) {attributeGetTypes [#1]}
  n37 = ScalarParameter() {defaultValue=-0.5, parameterName=MinStagger_Amt}
  n38 = ScalarParameter() {defaultValue=0.5, parameterName=MaxStagger_Amt}
  n39 = VertexColor()
  n40 = Add(A=Multiply(A=LinearInterpolate(A=@n37, B=@n38, Alpha=@n39.G), B=@n23), B=@n23)
  n41 = SphereMask(A=@n30.RGB, B=Add(A=@n2, B=@n25), Radius=Multiply(A=@n31, B=Clamp(None=@n40)), Hardness=Constant())
  n42 = ChannelMaskParameter(None=FunctionInput() {inputName=GMap}) {maskChannel=Green, parameterName=GMap_Channel, defaultValue=<rgba 0 1 0 0>}
  n43 = Subtract(A=@n41, B=Multiply(A=OneMinus(None=@n42.R), B=ScalarParameter() {defaultValue=0.25, parameterName=GMap Influence}) {constB=0.5})
  n44 = ScalarParameter() {defaultValue=-0.05, parameterName=SamplePos}
  n45 = Multiply(
   A=SphereMask(
    A=@n43
    B=@n44
    Radius=ScalarParameter() {defaultValue=0.015, parameterName=SampleRadius}
    Hardness=ScalarParameter() {parameterName=SampleHardness}
   )
   B=VectorParameter() {parameterName=EmissiveEdgeColor, defaultValue=<rgba 1 0.5903 0 1>}.RGB
  )
  n46 = ScalarParameter() {defaultValue=5, parameterName=EmissiveEdgeStrength}
  n47 = NamedRerouteUsage()
  n48 = Multiply(
   A=Multiply(A=Ceil(None=@n23), B=Multiply(A=@n47, B=@n24))
   B=VectorParameter() {parameterName=EmissiveEdgeColor, defaultValue=<rgba 1 0.5903 0 1>}.RGB
  )
  n49 = SphereMask(
   A=@n43
   B=@n44
   Radius=ScalarParameter() {defaultValue=0.125, parameterName=PatternGradientSampleRadius}
   Hardness=ScalarParameter() {parameterName=SampleHardness}
  )
  n50 = LinearInterpolate(
   A=ScalarParameter() {parameterName=PatternMinBrightness}
   B=ScalarParameter() {defaultValue=10, parameterName=PatternMaxBrightness}
   Alpha=@n49
  )
  n51 = Power(Base=Add(A=Multiply(A=@n45, B=@n46), B=Multiply(A=@n48, B=@n50)))
  n52 = FunctionInput(Preview=Constant4Vector().RGBA) {bUsePreviewValueAsDefault=true, inputName=GoldMask, inputType=FunctionInput_Vector4}
  n53 = LinearInterpolate(
   A=@n51
   B=Multiply(A=@n51, B=ChannelMaskParameter(None=@n52) {parameterName=GoldMask}.R)
   Alpha=ScalarParameter() {parameterName=UseGoldMask}
  )
  n54 = Add(A=@n36.EmissiveColor, B=Add(A=Multiply(A=@n24, B=@n53), B=@n34))
  n55 = Clamp(None=Ceil(None=@n41))
  FunctionOutput(None=BlendMaterialAttributes(
    A=SetMaterialAttributes!(MaterialAttributes=@n27.MaterialAttributes, EmissiveColor=@n35) {attributeSetTypes [#1]}
    B=SetMaterialAttributes(MaterialAttributes=@n36.MaterialAttributes, EmissiveColor=@n54) {attributeSetTypes [#1]}
    Alpha=@n55
   ) {vertexAttributeBlendType=UseA})
  FunctionOutput(None=@n55) {outputName=GoldMask}
  LinearInterpolate(A=ScalarParameter() {defaultValue=-1, parameterName=MinOffset}, B=@n15, Alpha=ConstantBiasScale(None=@n17))
  n56 = ScalarParameter() {parameterName=MLJ_TransformCurve}
  n57 = Clamp(None=Divide(A=@n56) {constB=23})
  LinearInterpolate(A=@n23, B=@n57, Alpha=ScalarParameter() {parameterName=UseSlider})
  MaterialFunctionCall!(Base, Blend) {materialFunction=/Engine/Functions/Engine_MaterialFunctions03/Blends/Blend_HardLight}
  Multiply(A=@n57, B=ScalarParameter() {group=SuperLevelling_Mask, parameterName=SL_CurrentLevel, sortPriority=3})
  n58 = ScalarParameter() {defaultValue=0.1, parameterName=SpotThickness_Min}
  n59 = ScalarParameter() {defaultValue=0.35, parameterName=SpotThickness_Max}
  Multiply(A=SphereMask(
    A=@n14
    B=Constant2Vector()
    Radius=LinearInterpolate(A=@n58, B=@n59, Alpha=@n49)
    Hardness=ScalarParameter() {defaultValue=0.5, parameterName=SpotHardness}
   )) {constB=0}
  NamedRerouteDeclaration(None=@n2) {name=PreskinnedPos, nodeColor=<rgba 0.4062 0 1 1>}
  NamedRerouteDeclaration(None=ChannelMaskParameter(None=@n13) {maskChannel=Green, parameterName=NoiseChannel, defaultValue=<rgba 0 1 0 0>}.R) {name=Pattern, nodeColor=<rgba 0.9453 1 0 1>}
  n60 = Add(A=Floor(None=@n57), B=@n56)
  n61 = Time()
  n62 = ScalarParameter() {defaultValue=-99999, group=SuperLevelling_Mask, parameterName=SL_GameTimeFromBP, sortPriority=100}
  n63 = ScalarParameter() {defaultValue=2, group=SuperLevelling_Mask, parameterName=SL_Animation_Duration, sortPriority=101}
  n64 = Saturate(None=Divide(A=Subtract(A=@n61, B=@n62), B=Add(A=@n63) {constB=0}))
  n65 = LinearInterpolate(A=Subtract(A=@n60), B=@n60, Alpha=@n64)
  n66 = ScalarParameter() {group=SuperLevelling_Mask, parameterName=SL_StartAtLevel, sortPriority=3}
  n67 = ScalarParameter() {defaultValue=23, group=SuperLevelling_Mask, parameterName=SL_StopAtLevel, sortPriority=3}
  n68 = Divide(A=Subtract(A=Clamp(None=@n65, Min=@n66, Max=@n67), B=@n66), B=Subtract(A=@n67, B=@n66))
  n69 = Max(A=Ceil(None=ScalarParameter() {defaultValue=1, group=SuperLevelling_Mask, parameterName=SL_QuantizationSteps, sliderMax=100, sliderMin=1}))
  n70 = Divide(A=Floor(None=Multiply(A=@n68, B=@n69)), B=@n69)
  n71 = Saturate(None=Ceil(None=Subtract(A=@n69)))
  NamedRerouteDeclaration(None=Clamp(None=LinearInterpolate(A=@n68, B=@n70, Alpha=@n71))) {name=FloatSteps, nodeColor=<rgba 0 1 0.0156 1>}
  n72 = ScalarParameter() {defaultValue=1.33, group=SuperLevelling_Mask, parameterName=SL_Animation_EaseIn, sortPriority=103}
  n73 = Multiply(A=Subtract(A=Power(Base=@n64, Exp=@n72) {constExponent=1.33}) {constB=0.5}) {constB=2}
  n74 = ConstantBiasScale(None=@n73)
  NamedRerouteDeclaration(None=@n74) {name=EdgeGlowTime, nodeColor=<rgba 0.8984 1 0 1>}
  NamedRerouteUsage()
  OneMinus(None=@n64)
  OneMinus(None=Abs(None=@n73))
  Power(Base=@n74, Exp=ScalarParameter() {defaultValue=6, group=SuperLevelling_Mask, parameterName=SL_Animation_EaseInOut, sortPriority=103}) {constExponent=8}
  Reroute(None=ConstantBiasScale(None=ComponentMask(None=@n16.Vec3) {b=true}))
  WorldPosition()
 ]}
