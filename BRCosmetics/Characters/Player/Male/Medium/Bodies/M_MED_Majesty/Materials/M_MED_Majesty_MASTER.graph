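# Master character material for the male medium Majesty skin, using material attributes with per-expression shading models.
# Majesty PreFX output passes through optional face-accessory crystal and clear coat effects into a gold blend, whose switchable second layer can carry lion-eye glow.
# Static switches then layer cloth fuzz, silk and metal LUT effects from a shared SkinFX mask.
# A UV1-masked tattoo tints base color before the shared character PostFX.
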Material /BRCosmetics/Characters/Player/Male/Medium/Bodies/M_MED_Majesty/Materials/M_MED_Majesty_MASTER
graph {
 properties {
  bUseMaterialAttributes=true
  bUsedWithMeshDeformer=true
  bUsedWithMorphTargets=true
  bUsedWithSkeletalMesh=true
  pixelDepthOffsetMode=PDOM_Legacy
  shadingModel=MSM_FromMaterialExpression
  usedShadingModels=Default Lit | Clear Coat | Subsurface Profile | Cloth
 }
 nodes [
  n1 = TextureSample(Tex=TextureObjectParameter() {group=Skin FX, parameterName=SkinFX_Mask})
  n2 = StaticSwitchParameter(
   True=MaterialFunctionCall!(
    Mask=@n1.RGBA
    PreskinnedPositon=@n9.Pre skinned Local Position
    Gmap=@n9.Pre skinned Local Position
    GoldMask=@n8.GoldMask
    Material=@n9.Pre skinned Local Position
   ) {materialFunction=&/MF_Majesty_CrystalFX}.Result
   False=@n9.Result
  ) {parameterName=BisFaceAcc}
  n3 = StaticSwitchParameter(
   True=MaterialFunctionCall!(Alpha=StaticComponentMaskParameter(None=@n1.RGBA) {defaultR=true, group=Skin FX, parameterName=Clear Coat Channel}, In=@n2) {materialFunction=/Game/Weapons/WeaponSkins/MaterialFunction/Clearcoat}.Result
   False=@n2
  ) {group=Skin FX, parameterName=Use Clear Coat}
  n4 = GetMaterialAttributes(None=@n9.Result)
  n5 = TextureSampleParameter2D() {parameterName=LionEyeEmissive, texture=&T_BlacksRGB}
  n6 = Constant3Vector() {constant=<rgba 1 0.5175 0 0>}
  n7 = Power(Base=Multiply(A=Multiply(A=@n5.R, B=@n6), B=ScalarParameter() {defaultValue=5, parameterName=EyeGlowStrength}))
  n8 = MaterialFunctionCall!(
   Material_02=StaticSwitchParameter(
    True=@n3
    False=SetMaterialAttributes(MaterialAttributes=@n4.MaterialAttributes, Emissive Color=@n7) {attributeSetTypes [#1]}
   ) {parameterName=UseCrystalGold}
   GMap=@n9.Gmap
   PreskinnedPos=@n9.Gmap
   Local Normal
   GoldMask=@n1.RGBA
   Material_01=@n3
  ) {materialFunction=&/MLB_MajestyGoldBlend}
  n9 = MaterialFunctionCall!(GoldMask=@n8.GoldMask, TexCoords) {materialFunction=&/MF_Majesty_PreFX}
  n10 = StaticSwitchParameter(
   True=MaterialFunctionCall!(
    Alpha=StaticComponentMaskParameter(None=@n1.RGBA) {defaultR=true, group=Skin FX, parameterName=Cloth Channel}
    Gmap=@n9.Gmap
    Material=@n8.Result
   ) {materialFunction=/Game/Weapons/WeaponSkins/MaterialFunction/ClothFuzz}.Result
   False=@n8.Result
  ) {group=Skin FX, parameterName=Use Cloth fuzz}
  n11 = StaticSwitchParameter(
   True=MaterialFunctionCall!(Alpha=ChannelMaskParameter(None=@n1.RGBA) {group=Skin FX, parameterName=Silk Channel}.R, Material=@n10) {materialFunction=/Game/Weapons/WeaponSkins/MaterialFunction/MF_Silk}.Result
   False=@n10
  ) {group=Skin FX, parameterName=Use Silk}
  n12 = GetMaterialAttributes(None=StaticSwitchParameter(
    True=MaterialFunctionCall(Mask=@n1.RGBA, Materials=@n11) {materialFunction=&MF_MetalsLUT}.Result
    False=@n11
   ) {group=Skin FX, parameterName=UseMetalLUT}) {attributeGetTypes [#2]}
  n13 = LinearInterpolate(
   A=@n12.Base Color
   B=VectorParameter() {parameterName=TattooColor, defaultValue=<rgba 0 0 0 1>}.RGB
   Alpha=TextureSampleParameter2D(UVs=TextureCoordinate() {coordinateIndex=1}) {parameterName=TattooMask, texture=&T_BlacksRGB}.G
  )
  n14 = ScalarParameter() {defaultValue=1, parameterName=TattooOpacity}
  MaterialFunctionCall(
   AO=@n9.AO
   Distance Blend=@n9.AO
   Fresnel=@n9.AO
   Gmap=@n9.AO
   Pre Skinned Local Position=@n9.AO
   Material=SetMaterialAttributes!(MaterialAttributes=@n12.MaterialAttributes, Base Color=LinearInterpolate(A=@n12.Base Color, B=@n13, Alpha=@n14)) {attributeSetTypes [#2]}
  ) {materialFunction=&BaseFN_CharacterFunction_PostFX}
  MaterialFunctionCall() {materialFunction=&BaseFN_CharacterFunction_PreFX}
  n15 = MaterialFunctionCall!(TexCoords) {materialFunction=&/MF_MajestyBaseGold}
  n16 = MaterialFunctionCall(Float3=@n15.M) {desc="R = AO\r\nG = GMap\r\nB = Skin", materialFunction=&BreakOutFloat3Components}
  n17 = MaterialFunctionCall!(Material=@n15.Result, Diffuse=@n15.Result, GMap=@n16.G) {materialFunction=/Game/Packages/Fortress_SharedMaterials/Base_Material_Functions/BaseCharacterShaderFunctions/BaseFN_CharacterGMapColor}
  n18 = Transform(None=@n15.Normal Map)
  MaterialFunctionCall!(
   Material=MaterialFunctionCall!(
    Material=@n17.Result
    Diffuse=@n17.Result
    BalanceDistance
    Normals=@n15.Normal Map
    AO=@n16.R
    Fresnel=Fresnel(Normal=@n18) {baseReflectFraction=0, exponent=1}
   ) {materialFunction=/Game/Packages/Fortress_SharedMaterials/Base_Material_Functions/BaseCharacterShaderFunctions/BaseFN_CharacterEmissive}.Result
   AO=@n16.R
   Skin Mask=@n16.B
  ) {materialFunction=/Game/Packages/Fortress_SharedMaterials/Base_Material_Functions/BaseCharacterShaderFunctions/BaseFN_CharacterSkin}
  Reroute()
  TextureObjectParameter() {group=Skin FX, parameterName=SkinFX_Mask}
  WorldPosition() {worldPositionShaderOffset=WPT_ExcludeAllShaderOffsets}
 ]
}
