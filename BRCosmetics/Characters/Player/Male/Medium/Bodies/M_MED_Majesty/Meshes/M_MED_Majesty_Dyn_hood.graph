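# Physics asset for the dynamic hood and drawstrings on the medium male Majesty body cosmetic, built from capsule bodies.
# Six hood chains start at small MSTR anchor bodies tied to spine_05 by free-rotation constraints, with limited swing and tight twist on the flap joints.
# Both four-segment string chains hang from the front-top hood bodies through fully angular-locked MSTR roots.
# Three capsules on each of spine_03 to spine_05 form the torso shape.
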PhysicsAsset /BRCosmetics/Characters/Player/Male/Medium/Bodies/M_MED_Majesty/Meshes/M_MED_Majesty_Dyn_hood
bodies {
 dyn_hood_bk_bott_01 [{shapeType=Capsule, length=5.0573, radius=3.5688, center=<2.6387 0 0>, rotation=<rot 90 0 0>}]
 dyn_hood_bk_bott_02 [{shapeType=Capsule, length=5.0573, radius=3.5688, center=<2.6387 0 0>, rotation=<rot 90 0 0>}]
 dyn_hood_bk_bott_MSTR [{shapeType=Capsule, length=0.505, radius=0.505}]
 dyn_hood_bk_top [{shapeType=Capsule, length=4.3698, radius=3.4751, center=<0.7529 0 0.0002>, rotation=<rot 90 0 0>}]
 dyn_hood_bk_top_MSTR [{shapeType=Capsule, length=0.505, radius=0.505}]
 dyn_hood_fr_top_l [{shapeType=Capsule, length=4.3698, radius=3.4751, center=<0.7529 0 0.0002>, rotation=<rot 90 0 0>}]
 dyn_hood_fr_top_l_MSTR [{shapeType=Capsule, length=0.505, radius=0.505}]
 dyn_hood_fr_top_r [{shapeType=Capsule, length=4.3698, radius=3.4751, center=<0.7529 0 0.0002>, rotation=<rot 90 0 0>}]
 dyn_hood_fr_top_r_MSTR [{shapeType=Capsule, length=0.505, radius=0.505}]
 dyn_hood_l_top [{shapeType=Capsule, length=4.3698, radius=3.4751, center=<0.7529 0 0.0002>, rotation=<rot 90 0 0>}]
 dyn_hood_l_top_MSTR [{shapeType=Capsule, length=0.505, radius=0.505}]
 dyn_hood_r_top [{shapeType=Capsule, length=4.3698, radius=3.4751, center=<0.7529 0 0.0002>, rotation=<rot 90 0 0>}]
 dyn_hood_r_top_MSTR [{shapeType=Capsule, length=0.505, radius=0.505}]
 dyn_string_l_01 [{shapeType=Capsule, length=3.1511, radius=0.6313, center=<2.1116 0 0.5>, rotation=<rot 90 0 0>}]
 dyn_string_l_02 [{shapeType=Capsule, length=4.3698, radius=0.9751, center=<2.5344 0 0.5>, rotation=<rot 90 0 0>}]
 dyn_string_l_03 [{shapeType=Capsule, length=4.3698, radius=0.9751, center=<2.5343 0 0.3>, rotation=<rot 90 0 0>}]
 dyn_string_l_04 [{shapeType=Capsule, length=2.9011, radius=0.9751, center=<2.2254 0.0001 0>, rotation=<rot 90 0 0>}]
 dyn_string_l_MSTR [{shapeType=Capsule, length=0.505, radius=0.505}]
 dyn_string_r_01 [{shapeType=Capsule, length=3.1511, radius=0.6313, center=<2.1116 0 0.5>, rotation=<rot 90 0 0>}]
 dyn_string_r_02 [{shapeType=Capsule, length=4.3698, radius=0.9751, center=<2.5344 0 0.5>, rotation=<rot 90 0 0>}]
 dyn_string_r_03 [{shapeType=Capsule, length=4.3698, radius=0.9751, center=<2.5343 0 0.3>, rotation=<rot 90 0 0>}]
 dyn_string_r_04 [{shapeType=Capsule, length=2.9011, radius=0.9751, center=<2.2254 0.0001 0>, rotation=<rot 90 0 0>}]
 dyn_string_r_MSTR [{shapeType=Capsule, length=0.505, radius=0.505}]
 spine_03 [
  {shapeType=Capsule, length=7.135, radius=13.5, center=<2.8663 -1.7031 0>}
  {shapeType=Capsule, length=9.6663, radius=2.5312, center=<2.5808 11.2231 -7.1373>, rotation=<rot 87.9873 -3.7957 3.7472>}
  {shapeType=Capsule, length=9.6663, radius=2.5312, center=<2.9412 11.0888 6.8864>, rotation=<rot 86.5651 176.2063 -176.401>}
 ]
 spine_04 [
  {shapeType=Capsule, length=6.635, radius=14, center=<2.8663 -1.7031 0>}
  {shapeType=Capsule, length=9.6663, radius=2.5312, center=<2.5808 11.2231 -7.1373>, rotation=<rot 87.9873 -3.7957 3.7472>}
  {shapeType=Capsule, length=9.6663, radius=2.5312, center=<2.9412 11.0888 6.8864>, rotation=<rot 86.5651 176.2063 -176.401>}
 ]
 spine_05 [
  {shapeType=Capsule, length=7.4163, radius=12.7812, center=<-1.4221 -0.4181 0>}
  {shapeType=Capsule, length=9.6663, radius=2.5312, center=<2.1983 9.2162 7.2453>, rotation=<rot 86.554 176.2055 -164.2026>}
  {shapeType=Capsule, length=9.6663, radius=2.5312, center=<3.1938 8.6279 -7.8807>, rotation=<rot 84.8493 -30.0902 -10.5706>}
 ]
}
constraints [
 {bone1Name=dyn_hood_fr_top_l, bone2Name=dyn_hood_fr_top_l_MSTR, swing1LimitDegrees=20, swing2LimitDegrees=20, twistLimitDegrees=2}
 {
  bone1Name=dyn_string_l_MSTR
  bone2Name=dyn_hood_fr_top_l
  swing1LimitDegrees=45
  swing1Motion=Locked
  swing2LimitDegrees=45
  swing2Motion=Locked
  twistLimitDegrees=45
  twistMotion=Locked
 }
 {bone1Name=dyn_string_l_01, bone2Name=dyn_string_l_MSTR, swing1LimitDegrees=12, swing2LimitDegrees=12, twistLimitDegrees=2}
 {bone1Name=dyn_string_l_02, bone2Name=dyn_string_l_01, swing1LimitDegrees=12, swing2LimitDegrees=12, twistLimitDegrees=2}
 {bone1Name=dyn_string_l_03, bone2Name=dyn_string_l_02, swing1LimitDegrees=12, swing2LimitDegrees=12, twistLimitDegrees=2}
 {bone1Name=dyn_string_l_04, bone2Name=dyn_string_l_03, swing1LimitDegrees=12, swing2LimitDegrees=12, twistLimitDegrees=2}
 {bone1Name=dyn_hood_l_top, bone2Name=dyn_hood_l_top_MSTR, swing1LimitDegrees=20, swing2LimitDegrees=20, twistLimitDegrees=2}
 {bone1Name=dyn_hood_bk_top, bone2Name=dyn_hood_bk_top_MSTR, swing1LimitDegrees=20, swing2LimitDegrees=20, twistLimitDegrees=2}
 {bone1Name=dyn_hood_bk_bott_01, bone2Name=dyn_hood_bk_bott_MSTR, swing1LimitDegrees=10, swing2LimitDegrees=20, twistLimitDegrees=2}
 {bone1Name=dyn_hood_bk_bott_02, bone2Name=dyn_hood_bk_bott_01, swing1LimitDegrees=10, swing2LimitDegrees=20, twistLimitDegrees=2}
 {bone1Name=dyn_hood_fr_top_r, bone2Name=dyn_hood_fr_top_r_MSTR, swing1LimitDegrees=20, swing2LimitDegrees=20, twistLimitDegrees=2}
 {
  bone1Name=dyn_string_r_MSTR
  bone2Name=dyn_hood_fr_top_r
  swing1LimitDegrees=45
  swing1Motion=Locked
  swing2LimitDegrees=45
  swing2Motion=Locked
  twistLimitDegrees=45
  twistMotion=Locked
 }
 {bone1Name=dyn_string_r_01, bone2Name=dyn_string_r_MSTR, swing1LimitDegrees=12, swing2LimitDegrees=12, twistLimitDegrees=2}
 {bone1Name=dyn_string_r_02, bone2Name=dyn_string_r_01, swing1LimitDegrees=12, swing2LimitDegrees=12, twistLimitDegrees=2}
 {bone1Name=dyn_string_r_03, bone2Name=dyn_string_r_02, swing1LimitDegrees=12, swing2LimitDegrees=12, twistLimitDegrees=2}
 {bone1Name=dyn_string_r_04, bone2Name=dyn_string_r_03, swing1LimitDegrees=12, swing2LimitDegrees=12, twistLimitDegrees=2}
 {bone1Name=dyn_hood_r_top, bone2Name=dyn_hood_r_top_MSTR, swing1LimitDegrees=20, swing2LimitDegrees=20, twistLimitDegrees=2}
 {
  bone1Name=dyn_hood_fr_top_l_MSTR
  bone2Name=spine_05
  swing1LimitDegrees=45
  swing1Motion=Free
  swing2LimitDegrees=45
  swing2Motion=Free
  twistLimitDegrees=45
  twistMotion=Free
 }
 {
  bone1Name=dyn_hood_l_top_MSTR
  bone2Name=spine_05
  swing1LimitDegrees=45
  swing1Motion=Free
  swing2LimitDegrees=45
  swing2Motion=Free
  twistLimitDegrees=45
  twistMotion=Free
 }
 {
  bone1Name=dyn_hood_bk_top_MSTR
  bone2Name=spine_05
  swing1LimitDegrees=45
  swing1Motion=Free
  swing2LimitDegrees=45
  swing2Motion=Free
  twistLimitDegrees=45
  twistMotion=Free
 }
 {
  bone1Name=dyn_hood_bk_bott_MSTR
  bone2Name=spine_05
  swing1LimitDegrees=45
  swing1Motion=Free
  swing2LimitDegrees=45
  swing2Motion=Free
  twistLimitDegrees=45
  twistMotion=Free
 }
 {
  bone1Name=dyn_hood_fr_top_r_MSTR
  bone2Name=spine_05
  swing1LimitDegrees=45
  swing1Motion=Free
  swing2LimitDegrees=45
  swing2Motion=Free
  twistLimitDegrees=45
  twistMotion=Free
 }
 {
  bone1Name=dyn_hood_r_top_MSTR
  bone2Name=spine_05
  swing1LimitDegrees=45
  swing1Motion=Free
  swing2LimitDegrees=45
  swing2Motion=Free
  twistLimitDegrees=45
  twistMotion=Free
 }
 {
  bone1Name=spine_05
  bone2Name=spine_04
  swing1LimitDegrees=45
  swing1Motion=Free
  swing2LimitDegrees=45
  swing2Motion=Free
  twistLimitDegrees=45
  twistMotion=Free
 }
 {
  bone1Name=spine_04
  bone2Name=spine_03
  swing1LimitDegrees=45
  swing1Motion=Free
  swing2LimitDegrees=45
  swing2Motion=Free
  twistLimitDegrees=45
  twistMotion=Free
 }
]
