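# Physics asset for the dynamic jacket on the male medium Majesty body, with capsules on pelvis, spine and thighs.
# Nine three-capsule chains around front, sides and back hang from small _MSTR bodies constrained to spine_04_aimer.
# Chains have tight twist limits, and most roots lock twist and swing1, swinging on one axis.
# Free-angle constraints link front tips to side tips, plus right outer-back to right side, with no left counterpart.
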PhysicsAsset /BRCosmetics/Characters/Player/Male/Medium/Bodies/M_MED_Majesty/Meshes/M_MED_Majesty_Dyn_jacket
bodies {
 dyn_jacket_bk_01 [{shapeType=Capsule, length=14.2188, radius=8, center=<9.0432 -0.0004 7>, rotation=<rot 90 0 0>}]
 dyn_jacket_bk_01_MSTR [{shapeType=Capsule, length=1, radius=1}]
 dyn_jacket_bk_02 [{shapeType=Capsule, length=6.9688, radius=6, center=<3.5594 -0.0002 6>, rotation=<rot 90 0 0>}]
 dyn_jacket_bk_03 [{shapeType=Capsule, length=6.875, radius=6, center=<3.5594 -0.0002 6>, rotation=<rot 90 0 0>}]
 dyn_jacket_bk_l_01 [{shapeType=Capsule, length=14.2188, radius=8, center=<9.0432 -0.0004 7>, rotation=<rot 90 0 0>}]
 dyn_jacket_bk_l_01_MSTR [{shapeType=Capsule, length=1, radius=1}]
 dyn_jacket_bk_l_02 [{shapeType=Capsule, length=6.7812, radius=6, center=<3.5594 -0.0002 6>, rotation=<rot 90 0 0>}]
 dyn_jacket_bk_l_03 [{shapeType=Capsule, length=6.7812, radius=6, center=<3.5594 -0.0002 6>, rotation=<rot 90 0 0>}]
 dyn_jacket_bk_r_01 [{shapeType=Capsule, length=13.9688, radius=8, center=<9.0432 -0.0004 7>, rotation=<rot 90 0 0>}]
 dyn_jacket_bk_r_01_MSTR [{shapeType=Capsule, length=1, radius=1}]
 dyn_jacket_bk_r_02 [{shapeType=Capsule, length=6.5625, radius=6, center=<3.5594 -0.0002 6>, rotation=<rot 90 0 0>}]
 dyn_jacket_bk_r_03 [{shapeType=Capsule, length=6.8125, radius=6, center=<3.5594 -0.0002 6>, rotation=<rot 90 0 0>}]
 dyn_jacket_bkout_l_01 [{shapeType=Capsule, length=14.2188, radius=8, center=<9.0432 -0.0004 7>, rotation=<rot 90 0 0>}]
 dyn_jacket_bkout_l_01_MSTR [{shapeType=Capsule, length=1, radius=1}]
 dyn_jacket_bkout_l_02 [{shapeType=Capsule, length=7.0938, radius=5, center=<3.5594 -0.0002 4.5>, rotation=<rot 90 0 0>}]
 dyn_jacket_bkout_l_03 [{shapeType=Capsule, length=6.75, radius=5.3438, center=<3.5594 -0.0002 4.5>, rotation=<rot 90 0 0>}]
 dyn_jacket_bkout_r_01 [{shapeType=Capsule, length=14.2188, radius=8, center=<9.0432 -0.0004 7>, rotation=<rot 90 0 0>}]
 dyn_jacket_bkout_r_01_MSTR [{shapeType=Capsule, length=1, radius=1}]
 dyn_jacket_bkout_r_02 [{shapeType=Capsule, length=7, radius=5.0938, center=<3.5594 -0.0002 4.5>, rotation=<rot 90 0 0>}]
 dyn_jacket_bkout_r_03 [{shapeType=Capsule, length=6.6562, radius=5.4375, center=<3.5594 -0.0002 4.5>, rotation=<rot 90 0 0>}]
 dyn_jacket_fr_l_01 [{shapeType=Capsule, length=14.2188, radius=5, center=<9.0432 -0.0004 4.5>, rotation=<rot 90 0 0>}]
 dyn_jacket_fr_l_01_MSTR [{shapeType=Capsule, length=1, radius=1}]
 dyn_jacket_fr_l_02 [{shapeType=Capsule, length=7.0938, radius=5, center=<3.5594 -0.0002 4.5>, rotation=<rot 90 0 0>}]
 dyn_jacket_fr_l_03 [{shapeType=Capsule, length=7.0938, radius=5, center=<3.5594 -0.0002 4.5>, rotation=<rot 90 0 0>}]
 dyn_jacket_fr_r_01 [{shapeType=Capsule, length=14.2188, radius=5, center=<9.0432 -0.0004 4.5>, rotation=<rot 90 0 0>}]
 dyn_jacket_fr_r_01_MSTR [{shapeType=Capsule, length=1, radius=1}]
 dyn_jacket_fr_r_02 [{shapeType=Capsule, length=7.0938, radius=5, center=<3.5594 -0.0002 4.5>, rotation=<rot 90 0 0>}]
 dyn_jacket_fr_r_03 [{shapeType=Capsule, length=7.0938, radius=5, center=<3.5594 -0.0002 4.5>, rotation=<rot 90 0 0>}]
 dyn_jacket_l_01 [{shapeType=Capsule, length=14.2188, radius=5, center=<9.0432 -0.0004 4.5>, rotation=<rot 90 0 0>}]
 dyn_jacket_l_01_MSTR [{shapeType=Capsule, length=1, radius=1}]
 dyn_jacket_l_02 [{shapeType=Capsule, length=7.0938, radius=5, center=<3.5594 -0.0002 4.5>, rotation=<rot 90 0 0>}]
 dyn_jacket_l_03 [{shapeType=Capsule, length=6.1875, radius=5.9062, center=<3.5594 -0.0002 4.5>, rotation=<rot 90 0 0>}]
 dyn_jacket_r_01 [{shapeType=Capsule, length=14.2188, radius=5, center=<9.0432 -0.0004 4.5>, rotation=<rot 90 0 0>}]
 dyn_jacket_r_01_MSTR [{shapeType=Capsule, length=1, radius=1}]
 dyn_jacket_r_02 [{shapeType=Capsule, length=7.0938, radius=5, center=<3.5594 -0.0002 4.5>, rotation=<rot 90 0 0>}]
 dyn_jacket_r_03 [{shapeType=Capsule, length=6.1875, radius=5.9062, center=<3.5594 -0.0002 4.5>, rotation=<rot 90 0 0>}]
 pelvis [{shapeType=Capsule, length=5.5938, radius=13.0312, center=<-1.4216 -1.591 -0.0001>}]
 spine_01 [{shapeType=Capsule, length=6.9688, radius=11.6875, center=<2.0249 0.4134 0.0001>}]
 spine_02 [{shapeType=Capsule, length=6.375, radius=11.8438, center=<3.2502 0.3213 -0.0002>}]
 spine_03 [{shapeType=Capsule, length=6.7188, radius=11.75, center=<4.3728 0.0138 -0.0001>}]
 spine_04_aimer [{shapeType=Capsule, length=1, radius=1}]
 thigh_l [{shapeType=Capsule, length=28.7188, radius=9.8438, center=<-17.057 -0.2342 2.0844>, rotation=<rot 88.5265 -9.6056 -12.5396>}]
 thigh_r [{shapeType=Capsule, length=28.7188, radius=9.8438, center=<17.0851 0.0978 -1.834>, rotation=<rot 88.5261 -9.6055 167.4604>}]
}
constraints [
 {bone1Name=dyn_jacket_fr_l_02, bone2Name=dyn_jacket_fr_l_01, swing1LimitDegrees=30, swing2LimitDegrees=45, twistLimitDegrees=2}
 {bone1Name=dyn_jacket_l_02, bone2Name=dyn_jacket_l_01, swing1LimitDegrees=10, swing2LimitDegrees=45, twistLimitDegrees=2}
 {bone1Name=dyn_jacket_r_02, bone2Name=dyn_jacket_r_01, swing1LimitDegrees=10, swing2LimitDegrees=60, twistLimitDegrees=2}
 {bone1Name=dyn_jacket_fr_r_02, bone2Name=dyn_jacket_fr_r_01, swing1LimitDegrees=25, swing2LimitDegrees=60, twistLimitDegrees=2}
 {bone1Name=dyn_jacket_fr_l_01, bone2Name=dyn_jacket_fr_l_01_MSTR, swing1LimitDegrees=20, swing2LimitDegrees=45, twistLimitDegrees=2}
 {
  bone1Name=dyn_jacket_l_01
  bone2Name=dyn_jacket_l_01_MSTR
  swing1LimitDegrees=10
  swing1Motion=Locked
  swing2LimitDegrees=45
  twistLimitDegrees=2
  twistMotion=Locked
 }
 {
  bone1Name=dyn_jacket_r_01
  bone2Name=dyn_jacket_r_01_MSTR
  swing1LimitDegrees=10
  swing1Motion=Locked
  swing2LimitDegrees=45
  twistLimitDegrees=2
  twistMotion=Locked
 }
 {bone1Name=dyn_jacket_fr_r_01, bone2Name=dyn_jacket_fr_r_01_MSTR, swing1LimitDegrees=20, swing2LimitDegrees=45, twistLimitDegrees=2}
 {
  bone1Name=thigh_l
  bone2Name=pelvis
  swing1LimitDegrees=45
  swing1Motion=Free
  swing2LimitDegrees=45
  swing2Motion=Free
  twistLimitDegrees=45
  twistMotion=Free
 }
 {
  bone1Name=dyn_jacket_fr_r_01_MSTR
  bone2Name=spine_04_aimer
  swing1LimitDegrees=45
  swing1Motion=Free
  swing2LimitDegrees=45
  swing2Motion=Free
  twistLimitDegrees=45
  twistMotion=Free
 }
 {
  bone1Name=dyn_jacket_r_01_MSTR
  bone2Name=spine_04_aimer
  swing1LimitDegrees=45
  swing1Motion=Free
  swing2LimitDegrees=45
  swing2Motion=Free
  twistLimitDegrees=45
  twistMotion=Free
 }
 {
  bone1Name=dyn_jacket_fr_l_01_MSTR
  bone2Name=spine_04_aimer
  swing1LimitDegrees=45
  swing1Motion=Free
  swing2LimitDegrees=45
  swing2Motion=Free
  twistLimitDegrees=45
  twistMotion=Free
 }
 {
  bone1Name=dyn_jacket_l_01_MSTR
  bone2Name=spine_04_aimer
  swing1LimitDegrees=45
  swing1Motion=Free
  swing2LimitDegrees=45
  swing2Motion=Free
  twistLimitDegrees=45
  twistMotion=Free
 }
 {
  bone1Name=spine_01
  bone2Name=pelvis
  swing1LimitDegrees=45
  swing1Motion=Free
  swing2LimitDegrees=45
  swing2Motion=Free
  twistLimitDegrees=45
  twistMotion=Free
 }
 {
  bone1Name=spine_02
  bone2Name=spine_01
  swing1LimitDegrees=45
  swing1Motion=Free
  swing2LimitDegrees=45
  swing2Motion=Free
  twistLimitDegrees=45
  twistMotion=Free
 }
 {
  bone1Name=dyn_jacket_bk_r_01_MSTR
  bone2Name=spine_04_aimer
  swing1LimitDegrees=45
  swing1Motion=Free
  swing2LimitDegrees=45
  swing2Motion=Free
  twistLimitDegrees=45
  twistMotion=Free
 }
 {
  bone1Name=dyn_jacket_bk_r_01
  bone2Name=dyn_jacket_bk_r_01_MSTR
  swing1LimitDegrees=10
  swing1Motion=Locked
  swing2LimitDegrees=60
  twistLimitDegrees=2
  twistMotion=Locked
 }
 {bone1Name=dyn_jacket_bk_r_02, bone2Name=dyn_jacket_bk_r_01, swing1LimitDegrees=10, swing2LimitDegrees=45, twistLimitDegrees=2}
 {bone1Name=dyn_jacket_bk_r_03, bone2Name=dyn_jacket_bk_r_02, swing1LimitDegrees=10, swing2LimitDegrees=45, twistLimitDegrees=2}
 {
  bone1Name=dyn_jacket_bk_01_MSTR
  bone2Name=spine_04_aimer
  swing1LimitDegrees=45
  swing1Motion=Free
  swing2LimitDegrees=45
  swing2Motion=Free
  twistLimitDegrees=45
  twistMotion=Free
 }
 {bone1Name=dyn_jacket_bk_01, bone2Name=dyn_jacket_bk_01_MSTR, swing1LimitDegrees=10, swing2LimitDegrees=60, twistLimitDegrees=2}
 {bone1Name=dyn_jacket_bk_02, bone2Name=dyn_jacket_bk_01, swing1LimitDegrees=10, swing2LimitDegrees=45, twistLimitDegrees=2}
 {bone1Name=dyn_jacket_bk_03, bone2Name=dyn_jacket_bk_02, swing1LimitDegrees=10, swing2LimitDegrees=45, twistLimitDegrees=2}
 {bone1Name=dyn_jacket_r_03, bone2Name=dyn_jacket_r_02, swing1LimitDegrees=10, swing2LimitDegrees=70, twistLimitDegrees=2}
 {bone1Name=dyn_jacket_fr_r_03, bone2Name=dyn_jacket_fr_r_02, swing1LimitDegrees=45, swing2LimitDegrees=70, twistLimitDegrees=2}
 {bone1Name=dyn_jacket_fr_l_03, bone2Name=dyn_jacket_fr_l_02, swing1LimitDegrees=20, swing2LimitDegrees=60, twistLimitDegrees=2}
 {bone1Name=dyn_jacket_l_03, bone2Name=dyn_jacket_l_02, swing1LimitDegrees=10, swing2LimitDegrees=45, twistLimitDegrees=2}
 {bone1Name=dyn_jacket_bk_l_03, bone2Name=dyn_jacket_bk_l_02, swing1LimitDegrees=10, swing2LimitDegrees=45, twistLimitDegrees=2}
 {bone1Name=dyn_jacket_bk_l_02, bone2Name=dyn_jacket_bk_l_01, swing1LimitDegrees=10, swing2LimitDegrees=45, twistLimitDegrees=2}
 {
  bone1Name=dyn_jacket_bk_l_01
  bone2Name=dyn_jacket_bk_l_01_MSTR
  swing1LimitDegrees=10
  swing1Motion=Locked
  swing2LimitDegrees=60
  twistLimitDegrees=2
  twistMotion=Locked
 }
 {
  bone1Name=dyn_jacket_bk_l_01_MSTR
  bone2Name=spine_04_aimer
  swing1LimitDegrees=45
  swing1Motion=Free
  swing2LimitDegrees=45
  swing2Motion=Free
  twistLimitDegrees=45
  twistMotion=Free
 }
 {
  bone1Name=spine_03
  bone2Name=spine_02
  swing1LimitDegrees=45
  swing1Motion=Free
  swing2LimitDegrees=45
  swing2Motion=Free
  twistLimitDegrees=45
  twistMotion=Free
 }
 {
  bone1Name=spine_04_aimer
  bone2Name=spine_03
  swing1LimitDegrees=45
  swing1Motion=Free
  swing2LimitDegrees=45
  swing2Motion=Free
  twistLimitDegrees=45
  twistMotion=Free
 }
 {
  bone1Name=dyn_jacket_bkout_l_01_MSTR
  bone2Name=spine_04_aimer
  swing1LimitDegrees=45
  swing1Motion=Free
  swing2LimitDegrees=45
  swing2Motion=Free
  twistLimitDegrees=45
  twistMotion=Free
 }
 {
  bone1Name=dyn_jacket_bkout_l_01
  bone2Name=dyn_jacket_bkout_l_01_MSTR
  swing1LimitDegrees=10
  swing1Motion=Locked
  swing2LimitDegrees=60
  twistLimitDegrees=2
  twistMotion=Locked
 }
 {bone1Name=dyn_jacket_bkout_l_02, bone2Name=dyn_jacket_bkout_l_01, swing1LimitDegrees=10, swing2LimitDegrees=45, twistLimitDegrees=2}
 {bone1Name=dyn_jacket_bkout_l_03, bone2Name=dyn_jacket_bkout_l_02, swing1LimitDegrees=10, swing2LimitDegrees=45, twistLimitDegrees=2}
 {
  bone1Name=dyn_jacket_bkout_r_01_MSTR
  bone2Name=spine_04_aimer
  swing1LimitDegrees=45
  swing1Motion=Free
  swing2LimitDegrees=45
  swing2Motion=Free
  twistLimitDegrees=45
  twistMotion=Free
 }
 {
  bone1Name=dyn_jacket_bkout_r_01
  bone2Name=dyn_jacket_bkout_r_01_MSTR
  swing1LimitDegrees=10
  swing1Motion=Locked
  swing2LimitDegrees=60
  twistLimitDegrees=2
  twistMotion=Locked
 }
 {bone1Name=dyn_jacket_bkout_r_02, bone2Name=dyn_jacket_bkout_r_01, swing1LimitDegrees=10, swing2LimitDegrees=60, twistLimitDegrees=2}
 {bone1Name=dyn_jacket_bkout_r_03, bone2Name=dyn_jacket_bkout_r_02, swing1LimitDegrees=10, swing2LimitDegrees=60, twistLimitDegrees=2}
 {
  bone1Name=thigh_r
  bone2Name=pelvis
  swing1LimitDegrees=45
  swing1Motion=Free
  swing2LimitDegrees=45
  swing2Motion=Free
  twistLimitDegrees=45
  twistMotion=Free
 }
 {
  bone1Name=dyn_jacket_r_03
  bone2Name=dyn_jacket_fr_r_03
  swing1LimitDegrees=45
  swing1Motion=Free
  swing2LimitDegrees=45
  swing2Motion=Free
  twistLimitDegrees=45
  twistMotion=Free
 }
 {
  bone1Name=dyn_jacket_l_03
  bone2Name=dyn_jacket_fr_l_03
  swing1LimitDegrees=45
  swing1Motion=Free
  swing2LimitDegrees=45
  swing2Motion=Free
  twistLimitDegrees=45
  twistMotion=Free
 }
 {
  bone1Name=dyn_jacket_bkout_r_03
  bone2Name=dyn_jacket_r_03
  swing1LimitDegrees=45
  swing1Motion=Free
  swing2LimitDegrees=45
  swing2Motion=Free
  twistLimitDegrees=45
  twistMotion=Free
 }
]
