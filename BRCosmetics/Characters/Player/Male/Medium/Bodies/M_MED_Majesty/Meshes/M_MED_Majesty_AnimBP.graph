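# Animation Blueprint for the M_MED_Majesty player cosmetic body, adding deformation and secondary motion to its jacket, hood and lion-head pieces.
# It copies the attached parent mesh's pose, then applies twist and sculpter Control Rigs and pose drivers.
# bRigidBodyEnabled selects between RigidBody jacket simulation and a constraint-based fallback, both starting from cached PreSim.
# Additive bone offsets and pelvis constraints correct the result using limb-rotation and movement-state variables like sliding.
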AnimBlueprint /BRCosmetics/Characters/Player/Male/Medium/Bodies/M_MED_Majesty/Meshes/M_MED_Majesty_AnimBP
target_skeleton=/Game/Characters/Player/Male/Medium/Bodies/M_MED_Majesty/Meshes/M_MED_Majesty_Skeleton
graphs [
 AnimGraph {nodes [
   "Anim Dynamics - Chain: dyn_jacket_bk_l_01_MSTR -> dyn_jacket_bk_l_02"(Alpha=1, ExternalForce="0,0,0") {
    bChain=true
    alphaBoolBlend {blendOption=Linear}
    boundBone {boneName=dyn_jacket_bk_l_01_MSTR}
    chainEnd {boneName=dyn_jacket_bk_l_02}
    physicsBodyDefinitions [
     {
      boundBone {boneName=dyn_jacket_bk_l_01_MSTR}
      boxExtents=<10 10 10>
      constraintSetup {angularLimitsMax=<30 30 30>, angularLimitsMin=<-30 -30 -30>, angularTarget=<0 0 0>}
      localJointOffset=<0 0 0>
     }
     {
      boundBone {boneName=dyn_jacket_bk_l_01}
      boxExtents=<10 10 10>
      constraintSetup {angularLimitsMax=<30 30 30>, angularLimitsMin=<-30 -30 -30>, angularTarget=<0 0 0>}
      localJointOffset=<0 0 0>
     }
     {
      boundBone {boneName=dyn_jacket_bk_l_02}
      boxExtents=<10 10 10>
      constraintSetup {angularLimitsMax=<30 30 30>, angularLimitsMin=<-30 -30 -30>, angularTarget=<0 0 0>}
      localJointOffset=<0 0 0>
     }
    ]
    sphericalLimits [{limitRadius=10, drivingBone {boneName=spine_01}, sphereLocalOffset=<0 0 0>}]
   }
   n1 = "Constraint - dyn_jacket_bkout_l_01 (P)"(Alpha=1, ConstraintWeights_0=1) {
    alphaBoolBlend {blendOption=Linear}
    alphaScaleBiasClamp {bInterpResult=true, interpSpeedIncreasing=15}
    boneToModify {boneName=dyn_jacket_bkout_l_01}
    constraintSetup [{targetBone {boneName=spine_04}}]
    constraintWeights [1]
   }
   n2 = "Constraint - dyn_jacket_bk_l_01 (P)"(ComponentPose=@n1, Alpha=1, ConstraintWeights_0=1) {
    alphaBoolBlend {blendOption=Linear}
    alphaScaleBiasClamp {bInterpResult=true, interpSpeedIncreasing=15}
    boneToModify {boneName=dyn_jacket_bk_l_01}
    constraintSetup [{targetBone {boneName=spine_04}}]
    constraintWeights [1]
   }
   n3 = "Constraint - dyn_jacket_bkout_r_01 (P)"(ComponentPose=@n2, Alpha=1, ConstraintWeights_0=1) {
    alphaBoolBlend {blendOption=Linear}
    alphaScaleBiasClamp {bInterpResult=true, interpSpeedIncreasing=15}
    boneToModify {boneName=dyn_jacket_bkout_r_01}
    constraintSetup [{targetBone {boneName=spine_04}}]
    constraintWeights [1]
   }
   n4 = "Constraint - dyn_jacket_bk_r_01 (P)"(ComponentPose=@n3, Alpha=1, ConstraintWeights_0=1) {
    alphaBoolBlend {blendOption=Linear}
    alphaScaleBiasClamp {bInterpResult=true, interpSpeedIncreasing=15}
    boneToModify {boneName=dyn_jacket_bk_r_01}
    constraintSetup [{targetBone {boneName=spine_04}}]
    constraintWeights [1]
   }
   n5 = Look At - Bone: dyn_jacket_bkout_l_01_MSTR(ComponentPose=@n4, Alpha=1) {
    alphaBoolBlend {blendOption=Linear}
    alphaScaleBiasClamp {bInterpResult=true, interpSpeedIncreasing=15}
    boneToModify {boneName=dyn_jacket_bkout_l_01_MSTR}
    lookAtLocation=<-26.877 -9.2601 -22.1178>
    lookAtTarget {boneReference {boneName=spine_05}}
    lookAt_Axis {axis=<1 0>}
   }
   n6 = Look At - Bone: dyn_jacket_bk_l_01_MSTR(ComponentPose=@n5, Alpha=1) {
    alphaBoolBlend {blendOption=Linear}
    alphaScaleBiasClamp {bInterpResult=true, interpSpeedIncreasing=15}
    boneToModify {boneName=dyn_jacket_bk_l_01_MSTR}
    lookAtLocation=<-42.8468 -21.1836 -9.8203>
    lookAtTarget {boneReference {boneName=spine_05}}
    lookAt_Axis {axis=<1 0>}
   }
   n7 = Look At - Bone: dyn_jacket_bkout_r_01_MSTR(ComponentPose=@n6, Alpha=1) {
    alphaBoolBlend {blendOption=Linear}
    alphaScaleBiasClamp {bInterpResult=true, interpSpeedIncreasing=15}
    boneToModify {boneName=dyn_jacket_bkout_r_01_MSTR}
    lookAtLocation=<-37.1848 4.3825 41.2554>
    lookAtTarget {boneReference {boneName=spine_05}}
    lookAt_Axis {axis=<1 0>}
   }
   Look At - Bone: dyn_jacket_bk_r_01_MSTR(ComponentPose=@n7, Alpha=1) {
    alphaBoolBlend {blendOption=Linear}
    alphaScaleBiasClamp {bInterpResult=true, interpSpeedIncreasing=15}
    boneToModify {boneName=dyn_jacket_bk_r_01_MSTR}
    lookAtLocation=<-40.0933 -24.4603 24.545>
    lookAtTarget {boneReference {boneName=spine_05}}
    lookAt_Axis {axis=<1 0>}
   }
   n8 = Get arm rotation_l() -> (arm rotation_l:double)
   n9 = "Transform (Modify) Bone - Bone: dyn_hood_l_top_MSTR"(Alpha=@n8, Translation="0,-0.5,3.5", Rotation="2.000,0.000,0.000", Scale="1,1,1") {
    rotationMode=BMM_Additive
    rotationSpace=BCS_BoneSpace
    translationMode=BMM_Additive
    translationSpace=BCS_BoneSpace
    alphaBoolBlend {blendOption=Linear}
    alphaScaleBiasClamp {interpSpeedIncreasing=40}
    boneToModify {boneName=dyn_hood_l_top_MSTR}
    rotation {pitch=2}
    translation {y=-0.5, z=3.5}
   }
   "Transform (Modify) Bone - Bone: dyn_hood_fr_top_l_MSTR"(ComponentPose=@n9, Alpha=@n8, Translation="0,-0.5,3", Rotation="2,0.000,0.000", Scale="1,1,1") {
    rotationMode=BMM_Additive
    rotationSpace=BCS_BoneSpace
    translationMode=BMM_Additive
    translationSpace=BCS_BoneSpace
    alphaBoolBlend {blendOption=Linear}
    alphaScaleBiasClamp {interpSpeedIncreasing=40}
    boneToModify {boneName=dyn_hood_fr_top_l_MSTR}
    rotation {pitch=2}
    translation {y=-0.5, z=3}
   }
   n10 = Get leg_l_rotation fwd() -> (leg_l_rotation fwd:double)
   n11 = "Transform (Modify) Bone - Bone: dyn_jacket_fr_l_01_MSTR"(
    ComponentPose=Local To Component(LocalPose=Use cached pose 'PreSim'())
    Alpha=@n10
    Translation="-0.125,-3.890,2.048"
    Rotation="25.286,-17.740,-35.732"
    Scale="1,1,1"
   ) {
    rotationMode=BMM_Additive
    rotationSpace=BCS_BoneSpace
    translationMode=BMM_Additive
    translationSpace=BCS_BoneSpace
    alphaBoolBlend {blendOption=Linear}
    alphaScaleBiasClamp {bInterpResult=true, interpSpeedDecreasing=5, interpSpeedIncreasing=150}
    boneToModify {boneName=dyn_jacket_fr_l_01_MSTR}
    rotation=<rot 25.286 -17.74 -35.732>
    translation=<-0.1251 -3.8901 2.0481>
   }
   n12 = "Transform (Modify) Bone - Bone: dyn_jacket_l_01_MSTR"(ComponentPose=@n11, Alpha=@n10, Translation="1.870,-3.761,0.608", Rotation="-27.020,-19.139,-0.674", Scale="1,1,1") {
    rotationMode=BMM_Additive
    rotationSpace=BCS_BoneSpace
    translationMode=BMM_Additive
    translationSpace=BCS_BoneSpace
    alphaBoolBlend {blendOption=Linear}
    alphaScaleBiasClamp {bInterpResult=true, interpSpeedDecreasing=5, interpSpeedIncreasing=30}
    boneToModify {boneName=dyn_jacket_l_01_MSTR}
    rotation=<rot -27.02 -19.139 -0.674>
    translation=<1.87 -3.761 0.608>
   }
   n13 = "Transform (Modify) Bone - Bone: dyn_jacket_bkout_l_01_MSTR"(ComponentPose=@n12, Alpha=@n10, Translation="3.218,-0.323,0.000", Rotation="37.385,-10.551,-0.152", Scale="1,1,1") {
    rotationMode=BMM_Additive
    rotationSpace=BCS_BoneSpace
    translationMode=BMM_Additive
    translationSpace=BCS_BoneSpace
    alphaBoolBlend {blendOption=Linear}
    alphaScaleBiasClamp {bInterpResult=true, interpSpeedDecreasing=5, interpSpeedIncreasing=30}
    boneToModify {boneName=dyn_jacket_bkout_l_01_MSTR}
    rotation=<rot 37.385 -10.551 -0.152>
    translation=<3.218 -0.323>
   }
   n14 = Get leg_r_rotation_fwd() -> (leg_r_rotation_fwd:double)
   n15 = "Transform (Modify) Bone - Bone: dyn_jacket_fr_r_01_MSTR"(ComponentPose=@n13, Alpha=@n14, Translation="2.254,1.060,2.340", Rotation="43.465,-23.958,46.459", Scale="1,1,1") {
    rotationMode=BMM_Additive
    rotationSpace=BCS_BoneSpace
    translationMode=BMM_Additive
    translationSpace=BCS_BoneSpace
    alphaBoolBlend {blendOption=Linear}
    alphaScaleBiasClamp {bInterpResult=true, interpSpeedDecreasing=5, interpSpeedIncreasing=5}
    boneToModify {boneName=dyn_jacket_fr_r_01_MSTR}
    rotation=<rot 43.465 -23.958 46.459>
    translation=<2.254 1.06 2.34>
   }
   n16 = "Transform (Modify) Bone - Bone: dyn_jacket_r_01_MSTR"(ComponentPose=@n15, Alpha=@n14, Translation="0.004,0.839,1.166", Rotation="23.824,5.853,2.368", Scale="1,1,1") {
    rotationMode=BMM_Additive
    rotationSpace=BCS_BoneSpace
    translationMode=BMM_Additive
    translationSpace=BCS_BoneSpace
    alphaBoolBlend {blendOption=Linear}
    alphaScaleBiasClamp {bInterpResult=true, interpSpeedDecreasing=5, interpSpeedIncreasing=5}
    boneToModify {boneName=dyn_jacket_r_01_MSTR}
    rotation=<rot 23.824 5.853 2.368>
    translation=<0.004 0.839 1.166>
   }
   n17 = RigidBody!(ComponentPose=@n16, Alpha, OverrideWorldGravity:Vector="0,0,-980", ExternalForce) -> (Pose) {
    bClampLinearTranslationLimitToRefPose=true
    bOverrideWorldGravity=true
    overridePhysicsAsset=&/M_MED_Majesty_Dyn_jacket
    simulationSpace=BaseBoneSpace
    alphaBoolBlend {blendOption=Linear}
    baseBoneRef {boneName=spine_02}
    componentAppliedLinearAccClamp=<3000 3000 3000>
    componentLinearAccScale {x=0.2}
    componentLinearVelScale {x=0.2}
    overrideWorldGravity {z=-980}
    simSpaceSettings {
     maxAngularAcceleration=500
     maxAngularVelocity=3
     maxLinearAcceleration=400
     maxLinearVelocity=250
     velocityScaleZ=0
     worldAlpha=1
     externalLinearDragV=<0.1 0.1>
    }
   }
   n18 = "Transform (Modify) Bone - Bone: dyn_jacket_fr_l_02"(ComponentPose=@n17, Alpha=Get Thigh_l() -> (Thigh_l:float), Translation="0,0,0", Rotation="0,10,0", Scale="1,1,1") {
    rotationMode=BMM_Additive
    rotationSpace=BCS_BoneSpace
    alphaBoolBlend {blendOption=Linear}
    boneToModify {boneName=dyn_jacket_fr_l_02}
    rotation {yaw=10}
   }
   n19 = Component To Local(ComponentPose=@n18)
   n20 = "Constraint - dyn_jacket_bk_r_01 (P)"(ComponentPose=Local To Component(LocalPose=Use cached pose 'PreSim'()), Alpha=1, ConstraintWeights_0=1) {
    alphaBoolBlend {blendOption=Linear}
    boneToModify {boneName=dyn_jacket_bk_r_01}
    constraintSetup [{targetBone {boneName=spine_04}}]
    constraintWeights [1]
   }
   n21 = "Constraint - dyn_jacket_bk_l_01 (P)"(ComponentPose=@n20, Alpha=1, ConstraintWeights_0=1) {
    alphaBoolBlend {blendOption=Linear}
    boneToModify {boneName=dyn_jacket_bk_l_01}
    constraintSetup [{targetBone {boneName=spine_04}}]
    constraintWeights [1]
   }
   n22 = "Constraint - dyn_jacket_bk_r_02 (P)"(ComponentPose=@n21, Alpha=1, ConstraintWeights_0=1) {
    alphaBoolBlend {blendOption=Linear}
    boneToModify {boneName=dyn_jacket_bk_r_02}
    constraintSetup [{targetBone {boneName=spine_01}}]
    constraintWeights [1]
   }
   n23 = "Constraint - dyn_jacket_bk_l_02 (P)"(ComponentPose=@n22, Alpha=1, ConstraintWeights_0=1) {
    alphaBoolBlend {blendOption=Linear}
    boneToModify {boneName=dyn_jacket_bk_l_02}
    constraintSetup [{targetBone {boneName=spine_01}}]
    constraintWeights [1]
   }
   n24 = "Constraint - dyn_jacket_bkout_l_01 (P)"(ComponentPose=@n23, Alpha=1, ConstraintWeights_0=1) {
    alphaBoolBlend {blendOption=Linear}
    boneToModify {boneName=dyn_jacket_bkout_l_01}
    constraintSetup [{targetBone {boneName=spine_04}}]
    constraintWeights [1]
   }
   n25 = "Constraint - dyn_jacket_bkout_r_01 (P)"(ComponentPose=@n24, Alpha=1, ConstraintWeights_0=1) {
    alphaBoolBlend {blendOption=Linear}
    boneToModify {boneName=dyn_jacket_bkout_r_01}
    constraintSetup [{targetBone {boneName=spine_04}}]
    constraintWeights [1]
   }
   n26 = "Constraint - dyn_jacket_bkout_l_02 (P)"(ComponentPose=@n25, Alpha=1, ConstraintWeights_0=1) {
    alphaBoolBlend {blendOption=Linear}
    boneToModify {boneName=dyn_jacket_bkout_l_02}
    constraintSetup [{targetBone {boneName=spine_01}}]
    constraintWeights [1]
   }
   n27 = "Constraint - dyn_jacket_bkout_r_02 (P)"(ComponentPose=@n26, Alpha=1, ConstraintWeights_0=1) {
    alphaBoolBlend {blendOption=Linear}
    boneToModify {boneName=dyn_jacket_bkout_r_02}
    constraintSetup [{targetBone {boneName=spine_01}}]
    constraintWeights [1]
   }
   n28 = "Constraint - dyn_jacket_bk_01 (P)"!(ComponentPose=@n27, bAlphaBoolEnabled:bool=True, ConstraintWeights_0=1) -> (Pose) {
    alphaInputType=Bool
    alphaBoolBlend {blendInTime=0.2, blendOption=Linear, blendOutTime=0.2}
    boneToModify {boneName=dyn_jacket_bk_01}
    constraintSetup [{targetBone {boneName=spine_01}}]
    constraintWeights [1]
   }
   n29 = "Constraint - dyn_jacket_bk_01 (P)"(ComponentPose=@n28, bAlphaBoolEnabled=True, ConstraintWeights_0=1) {
    alphaInputType=Bool
    alphaBoolBlend {blendInTime=0.2, blendOption=Linear, blendOutTime=0.9}
    boneToModify {boneName=dyn_jacket_bk_01}
    constraintSetup [{targetBone {boneName=spine_03}}]
    constraintWeights [1]
   }
   n30 = "Transform (Modify) Bone - Bone: dyn_jacket_bk_01"!(ComponentPose=@n29, bAlphaBoolEnabled:bool=True, Translation="0,0,1", Rotation="0,0,0", Scale="1,1,1") -> (Pose) {
    alphaInputType=Bool
    translationMode=BMM_Additive
    translationSpace=BCS_BoneSpace
    alphaBoolBlend {blendInTime=0.2, blendOption=Linear, blendOutTime=0.2}
    boneToModify {boneName=dyn_jacket_bk_01}
    translation {z=1}
   }
   n31 = "Transform (Modify) Bone - Bone: dyn_jacket_bk_01"(ComponentPose=@n30, bAlphaBoolEnabled=Get bIsPlayingEmote(), Translation="0,0,2", Rotation="0,0,0", Scale="1,1,1") {
    alphaInputType=Bool
    translationMode=BMM_Additive
    translationSpace=BCS_BoneSpace
    alphaBoolBlend {blendInTime=0.4, blendOption=Linear, blendOutTime=0.4}
    boneToModify {boneName=dyn_jacket_bk_01}
    translation {z=2}
   }
   n32 = "Transform (Modify) Bone - Bone: dyn_jacket_bk_l_02"(ComponentPose=@n31, Alpha=1, Translation="0,0,2", Rotation="0,0,0", Scale="1,1,1") {
    translationMode=BMM_Additive
    translationSpace=BCS_BoneSpace
    alphaBoolBlend {blendOption=Linear}
    boneToModify {boneName=dyn_jacket_bk_l_02}
    translation {z=2}
   }
   n33 = "Transform (Modify) Bone - Bone: dyn_jacket_bkout_l_02"(ComponentPose=@n32, Alpha=1, Translation="0,0,2", Rotation="0,0,0", Scale="1,1,1") {
    translationMode=BMM_Additive
    translationSpace=BCS_BoneSpace
    alphaBoolBlend {blendOption=Linear}
    boneToModify {boneName=dyn_jacket_bkout_l_02}
    translation {z=2}
   }
   n34 = "Transform (Modify) Bone - Bone: dyn_jacket_bkout_r_02"(ComponentPose=@n33, Alpha=1, Translation="0,3,3", Rotation="0,0,0", Scale="1,1,1") {
    translationMode=BMM_Additive
    translationSpace=BCS_BoneSpace
    alphaBoolBlend {blendOption=Linear}
    boneToModify {boneName=dyn_jacket_bkout_r_02}
    translation {y=3, z=3}
   }
   n35 = "Transform (Modify) Bone - Bone: dyn_jacket_bk_r_02"(ComponentPose=@n34, Alpha=1, Translation="0,0,2", Rotation="0,0,0", Scale="1,1,1") {
    translationMode=BMM_Additive
    translationSpace=BCS_BoneSpace
    alphaBoolBlend {blendOption=Linear}
    boneToModify {boneName=dyn_jacket_bk_r_02}
    translation {z=2}
   }
   n36 = Get bIsSkyDiving()
   n37 = "Constraint - dyn_jacket_bk_01 (P)"(ComponentPose=@n35, bAlphaBoolEnabled=@n36, ConstraintWeights_0=1) {
    alphaInputType=Bool
    alphaBoolBlend {blendOption=Linear}
    boneToModify {boneName=dyn_jacket_bk_01}
    constraintSetup [{targetBone {boneName=spine_01}}]
    constraintWeights [1]
   }
   n38 = "Transform (Modify) Bone - Bone: dyn_jacket_bk_02"!(ComponentPose=@n37, bAlphaBoolEnabled:bool=@n36, Translation="0,0,4", Rotation="0,0,0", Scale="1,1,1") -> (Pose) {
    alphaInputType=Bool
    rotationMode=BMM_Additive
    rotationSpace=BCS_BoneSpace
    translationMode=BMM_Additive
    translationSpace=BCS_BoneSpace
    alphaBoolBlend {blendInTime=0.2, blendOption=Linear, blendOutTime=0.2}
    boneToModify {boneName=dyn_jacket_bk_02}
    translation {z=4}
   }
   n39 = "Transform (Modify) Bone - Bone: dyn_jacket_bk_l_02"!(ComponentPose=@n38, bAlphaBoolEnabled:bool=@n36, Translation="0,0,4", Rotation="0,0,0", Scale="1,1,1") -> (Pose) {
    alphaInputType=Bool
    rotationMode=BMM_Additive
    rotationSpace=BCS_BoneSpace
    translationMode=BMM_Additive
    translationSpace=BCS_BoneSpace
    alphaBoolBlend {blendInTime=0.2, blendOption=Linear, blendOutTime=0.2}
    boneToModify {boneName=dyn_jacket_bk_l_02}
    translation {z=4}
   }
   n40 = "Transform (Modify) Bone - Bone: dyn_jacket_bk_r_02"!(ComponentPose=@n39, bAlphaBoolEnabled:bool=@n36, Translation="0,0,4", Rotation="0,0,0", Scale="1,1,1") -> (Pose) {
    alphaInputType=Bool
    rotationMode=BMM_Additive
    rotationSpace=BCS_BoneSpace
    translationMode=BMM_Additive
    translationSpace=BCS_BoneSpace
    alphaBoolBlend {blendInTime=0.2, blendOption=Linear, blendOutTime=0.2}
    boneToModify {boneName=dyn_jacket_bk_r_02}
    translation {z=4}
   }
   n41 = "Transform (Modify) Bone - Bone: dyn_jacket_bk_02"(ComponentPose=@n40, bAlphaBoolEnabled=True, Translation="0,0,3", Rotation="2,0,0", Scale="1,1,1") {
    alphaInputType=Bool
    rotationMode=BMM_Additive
    rotationSpace=BCS_BoneSpace
    translationMode=BMM_Additive
    translationSpace=BCS_BoneSpace
    alphaBoolBlend {blendInTime=0.2, blendOption=Linear, blendOutTime=0.5}
    boneToModify {boneName=dyn_jacket_bk_02}
    rotation {pitch=2}
    translation {z=3}
   }
   n42 = Get Thigh_l()
   n43 = "Transform (Modify) Bone - Bone: dyn_jacket_fr_l_01"(ComponentPose=@n41, Alpha=@n42, Translation="0,0,0", Rotation="0,-30,-20", Scale="1,1,1") {
    rotationMode=BMM_Additive
    rotationSpace=BCS_BoneSpace
    alphaBoolBlend {blendOption=Linear}
    boneToModify {boneName=dyn_jacket_fr_l_01}
    rotation {roll=-20, yaw=-30}
   }
   n44 = "Transform (Modify) Bone - Bone: dyn_jacket_fr_l_02"(ComponentPose=@n43, Alpha=@n42, Translation="0,0,0", Rotation="0,-10,0", Scale="1,1,1") {
    rotationMode=BMM_Additive
    rotationSpace=BCS_BoneSpace
    alphaBoolBlend {blendOption=Linear}
    boneToModify {boneName=dyn_jacket_fr_l_02}
    rotation {yaw=-10}
   }
   n45 = "Transform (Modify) Bone - Bone: dyn_jacket_l_01"(ComponentPose=@n44, Alpha=@n42, Translation="0,0,0", Rotation="0,-15,-10", Scale="1,1,1") {
    rotationMode=BMM_Additive
    rotationSpace=BCS_BoneSpace
    alphaBoolBlend {blendOption=Linear}
    boneToModify {boneName=dyn_jacket_l_01}
    rotation {roll=-10, yaw=-15}
   }
   n46 = Get Thigh_r() -> (Thigh_r:float)
   n47 = "Transform (Modify) Bone - Bone: dyn_jacket_fr_r_01"(ComponentPose=@n45, Alpha=@n46, Translation="0,0,0", Rotation="50,0,0", Scale="1,1,1") {
    rotationMode=BMM_Additive
    rotationSpace=BCS_BoneSpace
    alphaBoolBlend {blendOption=Linear}
    boneToModify {boneName=dyn_jacket_fr_r_01}
    rotation {pitch=50}
   }
   n48 = "Transform (Modify) Bone - Bone: dyn_jacket_r_01"(ComponentPose=@n47, Alpha=@n46, Translation="0,0,0", Rotation="15,0,0", Scale="1,1,1") {
    rotationMode=BMM_Additive
    rotationSpace=BCS_BoneSpace
    alphaBoolBlend {blendOption=Linear}
    boneToModify {boneName=dyn_jacket_r_01}
    rotation {pitch=15}
   }
   n49 = "Transform (Modify) Bone - Bone: dyn_jacket_fr_r_03"(ComponentPose=@n48, Alpha=@n46, Translation="0,0,0", Rotation="5,0,0", Scale="1,1,1") {
    rotationMode=BMM_Additive
    rotationSpace=BCS_BoneSpace
    alphaBoolBlend {blendOption=Linear}
    boneToModify {boneName=dyn_jacket_fr_r_03}
    rotation {pitch=5}
   }
   n50 = "Transform (Modify) Bone - Bone: dyn_jacket_fr_r_02"(ComponentPose=@n49, Alpha=@n46, Translation="0,0,0", Rotation="15,0,0", Scale="1,1,1") {
    rotationMode=BMM_Additive
    rotationSpace=BCS_BoneSpace
    alphaBoolBlend {blendOption=Linear}
    boneToModify {boneName=dyn_jacket_fr_r_02}
    rotation {pitch=15}
   }
   n51 = "Transform (Modify) Bone - Bone: dyn_jacket_fr_r_01"!(ComponentPose=@n50, bAlphaBoolEnabled:bool=True, Translation="0,0,0", Rotation="2,2,0", Scale="1,1,1") -> (Pose) {
    alphaInputType=Bool
    rotationMode=BMM_Additive
    rotationSpace=BCS_BoneSpace
    alphaBoolBlend {blendInTime=0.2, blendOption=Linear, blendOutTime=0.2}
    boneToModify {boneName=dyn_jacket_fr_r_01}
    rotation {pitch=2, yaw=2}
   }
   n52 = "Transform (Modify) Bone - Bone: dyn_jacket_r_01"!(ComponentPose=@n51, bAlphaBoolEnabled:bool=True, Translation="0,0,0", Rotation="2,2,0", Scale="1,1,1") -> (Pose) {
    alphaInputType=Bool
    rotationMode=BMM_Additive
    rotationSpace=BCS_BoneSpace
    alphaBoolBlend {blendInTime=0.2, blendOption=Linear, blendOutTime=0.2}
    boneToModify {boneName=dyn_jacket_r_01}
    rotation {pitch=2, yaw=2}
   }
   n53 = "Transform (Modify) Bone - Bone: dyn_jacket_fr_l_01"!(ComponentPose=@n52, bAlphaBoolEnabled:bool=True, Translation="0,0,0", Rotation="2,2,0", Scale="1,1,1") -> (Pose) {
    alphaInputType=Bool
    rotationMode=BMM_Additive
    rotationSpace=BCS_BoneSpace
    alphaBoolBlend {blendInTime=0.2, blendOption=Linear, blendOutTime=0.2}
    boneToModify {boneName=dyn_jacket_fr_l_01}
    rotation {pitch=2, yaw=2}
   }
   n54 = "Transform (Modify) Bone - Bone: dyn_jacket_l_01"!(ComponentPose=@n53, bAlphaBoolEnabled:bool=True, Translation="0,0,0", Rotation="2,2,0", Scale="1,1,1") -> (Pose) {
    alphaInputType=Bool
    rotationMode=BMM_Additive
    rotationSpace=BCS_BoneSpace
    alphaBoolBlend {blendInTime=0.2, blendOption=Linear, blendOutTime=0.2}
    boneToModify {boneName=dyn_jacket_l_01}
    rotation {pitch=2, yaw=2}
   }
   n55 = "Transform (Modify) Bone - Bone: dyn_jacket_bk_01"(ComponentPose=@n54, bAlphaBoolEnabled=True, Translation="0,0,3", Rotation="3,0,0", Scale="1,1,1") {
    alphaInputType=Bool
    rotationMode=BMM_Additive
    rotationSpace=BCS_BoneSpace
    translationMode=BMM_Additive
    translationSpace=BCS_BoneSpace
    alphaBoolBlend {blendInTime=0.2, blendOption=Linear, blendOutTime=0.2}
    boneToModify {boneName=dyn_jacket_bk_01}
    rotation {pitch=3}
    translation {z=3}
   }
   n56 = "Transform (Modify) Bone - Bone: dyn_jacket_bkout_r_01"!(ComponentPose=@n55, bAlphaBoolEnabled:bool=True, Translation="0,0,3", Rotation="3,0,0", Scale="1,1,1") -> (Pose) {
    alphaInputType=Bool
    rotationMode=BMM_Additive
    rotationSpace=BCS_BoneSpace
    translationMode=BMM_Additive
    translationSpace=BCS_BoneSpace
    alphaBoolBlend {blendInTime=0.2, blendOption=Linear, blendOutTime=0.2}
    boneToModify {boneName=dyn_jacket_bkout_r_01}
    rotation {pitch=3}
    translation {z=3}
   }
   n57 = Get jump_melee() -> (jump_melee:bool)
   n58 = "Transform (Modify) Bone - Bone: dyn_jacket_bkout_r_01_MSTR"!(ComponentPose=@n56, bAlphaBoolEnabled:bool=@n57, Translation="0,0,3", Rotation="1,0,0", Scale="1,1,1") -> (Pose) {
    alphaInputType=Bool
    rotationMode=BMM_Additive
    rotationSpace=BCS_BoneSpace
    translationMode=BMM_Additive
    translationSpace=BCS_BoneSpace
    alphaBoolBlend {blendInTime=0.2, blendOption=Linear, blendOutTime=0.2}
    boneToModify {boneName=dyn_jacket_bkout_r_01_MSTR}
    rotation {pitch=1}
    translation {z=3}
   }
   n59 = "Transform (Modify) Bone - Bone: dyn_jacket_bk_r_01_MSTR"!(ComponentPose=@n58, bAlphaBoolEnabled:bool=@n57, Translation="0,0,2", Rotation="1,0,0", Scale="1,1,1") -> (Pose) {
    alphaInputType=Bool
    rotationMode=BMM_Additive
    rotationSpace=BCS_BoneSpace
    translationMode=BMM_Additive
    translationSpace=BCS_BoneSpace
    alphaBoolBlend {blendInTime=0.2, blendOption=Linear, blendOutTime=0.2}
    boneToModify {boneName=dyn_jacket_bk_r_01_MSTR}
    rotation {pitch=1}
    translation {z=2}
   }
   n60 = "Transform (Modify) Bone - Bone: dyn_jacket_bk_r_01"!(ComponentPose=@n59, bAlphaBoolEnabled:bool=@n57, Translation="-1,0,2", Rotation="2,0,0", Scale="1,1,1") -> (Pose) {
    alphaInputType=Bool
    rotationMode=BMM_Additive
    rotationSpace=BCS_BoneSpace
    translationMode=BMM_Additive
    translationSpace=BCS_BoneSpace
    alphaBoolBlend {blendInTime=0.2, blendOption=Linear, blendOutTime=0.2}
    boneToModify {boneName=dyn_jacket_bk_r_01}
    rotation {pitch=2}
    translation {x=-1, z=2}
   }
   n61 = "Transform (Modify) Bone - Bone: dyn_jacket_fr_r_02"!(ComponentPose=@n60, bAlphaBoolEnabled:bool=@n57, Translation="0,0,2", Rotation="10,0,0", Scale="1,1,1") -> (Pose) {
    alphaInputType=Bool
    rotationMode=BMM_Additive
    rotationSpace=BCS_BoneSpace
    translationMode=BMM_Additive
    translationSpace=BCS_BoneSpace
    alphaBoolBlend {blendInTime=0.2, blendOption=Linear, blendOutTime=0.2}
    boneToModify {boneName=dyn_jacket_fr_r_02}
    rotation {pitch=10}
    translation {z=2}
   }
   n62 = "Transform (Modify) Bone - Bone: dyn_jacket_r_02"!(ComponentPose=@n61, bAlphaBoolEnabled:bool=@n57, Translation="0,0,2", Rotation="10,0,0", Scale="1,1,1") -> (Pose) {
    alphaInputType=Bool
    rotationMode=BMM_Additive
    rotationSpace=BCS_BoneSpace
    translationMode=BMM_Additive
    translationSpace=BCS_BoneSpace
    alphaBoolBlend {blendInTime=0.2, blendOption=Linear, blendOutTime=0.2}
    boneToModify {boneName=dyn_jacket_r_02}
    rotation {pitch=10}
    translation {z=2}
   }
   n63 = "Transform (Modify) Bone - Bone: dyn_jacket_bk_l_02"(ComponentPose=@n62, bAlphaBoolEnabled=@n57, Translation="0,0,2", Rotation="10,0,0", Scale="1,1,1") {
    alphaInputType=Bool
    rotationMode=BMM_Additive
    rotationSpace=BCS_BoneSpace
    translationMode=BMM_Additive
    translationSpace=BCS_BoneSpace
    alphaBoolBlend {blendInTime=0.2, blendOption=Linear, blendOutTime=0.2}
    boneToModify {boneName=dyn_jacket_bk_l_02}
    rotation {pitch=10}
    translation {z=2}
   }
   n64 = "Transform (Modify) Bone - Bone: dyn_jacket_r_01_MSTR"!(ComponentPose=@n63, bAlphaBoolEnabled:bool=@n57, Translation="0,0,2", Rotation="0,0,0", Scale="1,1,1") -> (Pose) {
    alphaInputType=Bool
    rotationMode=BMM_Additive
    rotationSpace=BCS_BoneSpace
    translationMode=BMM_Additive
    translationSpace=BCS_BoneSpace
    alphaBoolBlend {blendInTime=0.2, blendOption=Linear, blendOutTime=0.2}
    boneToModify {boneName=dyn_jacket_r_01_MSTR}
    translation {z=2}
   }
   n65 = Get bIsSliding()
   n66 = "Transform (Modify) Bone - Bone: dyn_jacket_bk_l_01"!(ComponentPose=@n64, bAlphaBoolEnabled:bool=@n65, Translation="0,0,3", Rotation="0,0,0", Scale="1,1,1") -> (Pose) {
    alphaInputType=Bool
    rotationMode=BMM_Additive
    rotationSpace=BCS_BoneSpace
    translationMode=BMM_Additive
    translationSpace=BCS_BoneSpace
    alphaBoolBlend {blendInTime=0.1, blendOption=Linear, blendOutTime=0.2}
    boneToModify {boneName=dyn_jacket_bk_l_01}
    translation {z=3}
   }
   n67 = "Transform (Modify) Bone - Bone: dyn_jacket_bkout_l_01"!(ComponentPose=@n66, bAlphaBoolEnabled:bool=@n65, Translation="0,0,3", Rotation="0,0,0", Scale="1,1,1") -> (Pose) {
    alphaInputType=Bool
    rotationMode=BMM_Additive
    rotationSpace=BCS_BoneSpace
    translationMode=BMM_Additive
    translationSpace=BCS_BoneSpace
    alphaBoolBlend {blendInTime=0.1, blendOption=Linear, blendOutTime=0.2}
    boneToModify {boneName=dyn_jacket_bkout_l_01}
    translation {z=3}
   }
   n68 = Component To Local(ComponentPose=@n67)
   n69 = Get bRigidBodyEnabled()
   n70 = "Transform (Modify) Bone - Bone: dyn_hood_bk_top_MSTR"(
    ComponentPose=Local To Component(LocalPose=Blend Poses by bool(BlendPose_0=@n19, BlendPose_1=@n68, bActiveValue=@n69) {blendType=Linear})
    Alpha=Get Head Rotation() -> (Head Rotation:double)
    Translation="0,0,0"
    Rotation="0,0,0"
    Scale="0.6,1,1"
   ) {
    rotationMode=BMM_Additive
    rotationSpace=BCS_BoneSpace
    scaleMode=BMM_Additive
    scaleSpace=BCS_BoneSpace
    translationMode=BMM_Additive
    translationSpace=BCS_BoneSpace
    alphaBoolBlend {blendOption=Linear}
    alphaScaleBiasClamp {bInterpResult=true, interpSpeedIncreasing=15}
    boneToModify {boneName=dyn_hood_bk_top_MSTR}
    scale {x=0.6}
   }
   n71 = "Transform (Modify) Bone - Bone: dyn_jacket_bk_r_01_MSTR"!(ComponentPose=@n70, Alpha:float=1, Rotation="10,5,1") -> (Pose) {
    rotationMode=BMM_Additive
    rotationSpace=BCS_BoneSpace
    scaleSpace=BCS_BoneSpace
    translationSpace=BCS_BoneSpace
    alphaBoolBlend {blendOption=Linear}
    alphaScaleBiasClamp {bInterpResult=true, interpSpeedIncreasing=15}
    boneToModify {boneName=dyn_jacket_bk_r_01_MSTR}
    rotation=<rot 10 5 1>
    scale {x=0.6}
   }
   n72 = "Transform (Modify) Bone - Bone: dyn_hood_corr_r"(ComponentPose=@n71, Alpha=1, Translation="-7,0,0", Rotation="0,0,0", Scale="1,1,1") {
    translationMode=BMM_Additive
    translationSpace=BCS_BoneSpace
    alphaBoolBlend {blendOption=Linear}
    boneToModify {boneName=dyn_hood_corr_r}
    translation {x=-7}
   }
   n73 = "Transform (Modify) Bone - Bone: dyn_hood_corr_l"(ComponentPose=@n72, Alpha=1, Translation="-7,0,0", Rotation="0,0,0", Scale="1,1,1") {
    translationMode=BMM_Additive
    translationSpace=BCS_BoneSpace
    alphaBoolBlend {blendOption=Linear}
    boneToModify {boneName=dyn_hood_corr_l}
    translation {x=-7}
   }
   n74 = "Transform (Modify) Bone - Bone: dyn_lion_backhead_l"(ComponentPose=@n73, Alpha=1, Translation="3,0,-2", Rotation="0,0,0", Scale="1,1,1") {
    translationMode=BMM_Additive
    translationSpace=BCS_BoneSpace
    alphaBoolBlend {blendOption=Linear}
    boneToModify {boneName=dyn_lion_backhead_l}
    translation {x=3, z=-2}
   }
   n75 = "Transform (Modify) Bone - Bone: dyn_lion_backhead_r"(ComponentPose=@n74, Alpha=1, Translation="0,0,0", Rotation="0,0,0", Scale="1,1,1") {
    translationMode=BMM_Additive
    translationSpace=BCS_BoneSpace
    alphaBoolBlend {blendOption=Linear}
    boneToModify {boneName=dyn_lion_backhead_r}
   }
   n76 = "Transform (Modify) Bone - Bone: dyn_hood_corr_front_r"(ComponentPose=@n75, Alpha=1, Translation="0,-1,0", Rotation="0,0,0", Scale="1,1,1") {
    translationMode=BMM_Additive
    translationSpace=BCS_BoneSpace
    alphaBoolBlend {blendOption=Linear}
    boneToModify {boneName=dyn_hood_corr_front_r}
    translation {y=-1}
   }
   n77 = "Transform (Modify) Bone - Bone: dyn_hood_corr_front_l"(ComponentPose=@n76, Alpha=1, Translation="0,-1,0", Rotation="0,0,0", Scale="1,1,1") {
    translationMode=BMM_Additive
    translationSpace=BCS_BoneSpace
    alphaBoolBlend {blendOption=Linear}
    boneToModify {boneName=dyn_hood_corr_front_l}
    translation {y=-1}
   }
   n78 = Get Alpha_jacket() -> (Alpha_jacket:double)
   n79 = "Constraint - dyn_jacket_fr_r_03 (P)"(ComponentPose=@n77, Alpha=@n78, ConstraintWeights_0=1) {
    alphaBoolBlend {blendOption=Linear}
    boneToModify {boneName=dyn_jacket_fr_r_03}
    constraintSetup [{targetBone {boneName=pelvis}}]
    constraintWeights [1]
   }
   n80 = "Constraint - dyn_jacket_fr_l_03 (P)"(ComponentPose=@n79, Alpha=@n78, ConstraintWeights_0=1) {
    alphaBoolBlend {blendOption=Linear}
    boneToModify {boneName=dyn_jacket_fr_l_03}
    constraintSetup [{targetBone {boneName=pelvis}}]
    constraintWeights [1]
   }
   n81 = "Constraint - dyn_jacket_r_03 (P)"(ComponentPose=@n80, Alpha=@n78, ConstraintWeights_0=1) {
    alphaBoolBlend {blendOption=Linear}
    boneToModify {boneName=dyn_jacket_r_03}
    constraintSetup [{targetBone {boneName=pelvis}}]
    constraintWeights [1]
   }
   n82 = "Constraint - dyn_jacket_l_03 (P)"(ComponentPose=@n81, Alpha=@n78, ConstraintWeights_0=1) {
    alphaBoolBlend {blendOption=Linear}
    boneToModify {boneName=dyn_jacket_l_03}
    constraintSetup [{targetBone {boneName=pelvis}}]
    constraintWeights [1]
   }
   n83 = "Constraint - dyn_jacket_fr_r_02 (P)"(ComponentPose=@n82, Alpha=@n78, ConstraintWeights_0=1) {
    alphaBoolBlend {blendOption=Linear}
    boneToModify {boneName=dyn_jacket_fr_r_02}
    constraintSetup [{targetBone {boneName=pelvis}}]
    constraintWeights [1]
   }
   n84 = "Constraint - dyn_jacket_fr_l_02 (P)"(ComponentPose=@n83, Alpha=@n78, ConstraintWeights_0=1) {
    alphaBoolBlend {blendOption=Linear}
    boneToModify {boneName=dyn_jacket_fr_l_02}
    constraintSetup [{targetBone {boneName=pelvis}}]
    constraintWeights [1]
   }
   n85 = "Constraint - dyn_jacket_r_02 (P)"(ComponentPose=@n84, Alpha=@n78, ConstraintWeights_0=1) {
    alphaBoolBlend {blendOption=Linear}
    boneToModify {boneName=dyn_jacket_r_02}
    constraintSetup [{targetBone {boneName=pelvis}}]
    constraintWeights [1]
   }
   n86 = "Constraint - dyn_jacket_l_02 (P)"(ComponentPose=@n85, Alpha=@n78, ConstraintWeights_0=1) {
    alphaBoolBlend {blendOption=Linear}
    boneToModify {boneName=dyn_jacket_l_02}
    constraintSetup [{targetBone {boneName=pelvis}}]
    constraintWeights [1]
   }
   n87 = "Constraint - dyn_jacket_fr_r_01 (P)"(ComponentPose=@n86, Alpha=@n78, ConstraintWeights_0=1) {
    alphaBoolBlend {blendOption=Linear}
    boneToModify {boneName=dyn_jacket_fr_r_01}
    constraintSetup [{targetBone {boneName=pelvis}}]
    constraintWeights [1]
   }
   n88 = "Constraint - dyn_jacket_fr_l_01 (P)"(ComponentPose=@n87, Alpha=@n78, ConstraintWeights_0=1) {
    alphaBoolBlend {blendOption=Linear}
    boneToModify {boneName=dyn_jacket_fr_l_01}
    constraintSetup [{targetBone {boneName=pelvis}}]
    constraintWeights [1]
   }
   n89 = "Constraint - dyn_jacket_r_01 (P)"(ComponentPose=@n88, Alpha=@n78, ConstraintWeights_0=1) {
    alphaBoolBlend {blendOption=Linear}
    boneToModify {boneName=dyn_jacket_r_01}
    constraintSetup [{targetBone {boneName=pelvis}}]
    constraintWeights [1]
   }
   n90 = "Constraint - dyn_jacket_l_01 (P)"(ComponentPose=@n89, Alpha=@n78, ConstraintWeights_0=1) {
    alphaBoolBlend {blendOption=Linear}
    boneToModify {boneName=dyn_jacket_l_01}
    constraintSetup [{targetBone {boneName=pelvis}}]
    constraintWeights [1]
   }
   n91 = Get bIsSliding()
   n92 = "Transform (Modify) Bone - Bone: dyn_jacket_bk_r_01_MSTR"!(ComponentPose=@n90, bAlphaBoolEnabled:bool=@n91, Translation:Vector="0,-1,2", Rotation="0,0,0", Scale:Vector="1,1,1") -> (Pose) {
    alphaInputType=Bool
    rotationMode=BMM_Additive
    rotationSpace=BCS_BoneSpace
    translationMode=BMM_Additive
    translationSpace=BCS_BoneSpace
    alphaBoolBlend {blendInTime=0.1, blendOption=QuarticInOut, blendOutTime=0.4}
    boneToModify {boneName=dyn_jacket_bk_r_01_MSTR}
    translation {y=-1, z=2}
   }
   n93 = "Transform (Modify) Bone - Bone: dyn_jacket_bkout_r_01_MSTR"(ComponentPose=@n92, bAlphaBoolEnabled=@n91, Translation="0,-2,2", Rotation="0,0,0", Scale="1,1,1") {
    alphaInputType=Bool
    translationMode=BMM_Additive
    translationSpace=BCS_BoneSpace
    alphaBoolBlend {blendInTime=0.1, blendOption=QuarticInOut, blendOutTime=0.4}
    boneToModify {boneName=dyn_jacket_bkout_r_01_MSTR}
    translation {y=-2, z=2}
   }
   n94 = "Transform (Modify) Bone - Bone: dyn_jacket_r_01_MSTR"(ComponentPose=@n93, bAlphaBoolEnabled=@n91, Translation="0,-1,0", Rotation="0,0,0", Scale="1,1,1") {
    alphaInputType=Bool
    rotationMode=BMM_Additive
    rotationSpace=BCS_BoneSpace
    translationMode=BMM_Additive
    translationSpace=BCS_BoneSpace
    alphaBoolBlend {blendInTime=0.1, blendOption=QuarticInOut, blendOutTime=0.4}
    boneToModify {boneName=dyn_jacket_r_01_MSTR}
    translation {y=-1}
   }
   n95 = "Transform (Modify) Bone - Bone: dyn_jacket_fr_r_01_MSTR"!(ComponentPose=@n94, bAlphaBoolEnabled:bool=@n91, Translation="3,-2,0", Rotation="0,0,0", Scale="1,1,1") -> (Pose) {
    alphaInputType=Bool
    rotationMode=BMM_Additive
    rotationSpace=BCS_BoneSpace
    translationMode=BMM_Additive
    translationSpace=BCS_BoneSpace
    alphaBoolBlend {blendInTime=0.1, blendOption=QuarticInOut, blendOutTime=0.4}
    boneToModify {boneName=dyn_jacket_fr_r_01_MSTR}
    translation=<3 -2>
   }
   n96 = "Transform (Modify) Bone - Bone: dyn_jacket_r_01"!(ComponentPose=@n95, Alpha:float=Get sliding gun() -> (sliding gun:double), Translation="0,0,10", Rotation="0,0,0", Scale="1,1,1") -> (Pose) {
    rotationMode=BMM_Additive
    rotationSpace=BCS_BoneSpace
    translationMode=BMM_Additive
    translationSpace=BCS_BoneSpace
    alphaBoolBlend {blendInTime=0.1, blendOption=QuarticInOut, blendOutTime=0.4}
    boneToModify {boneName=dyn_jacket_r_01}
    translation {z=10}
   }
   n97 = "Transform (Modify) Bone - Bone: dyn_jacket_r_01"!(ComponentPose=@n96, bAlphaBoolEnabled:bool=Get bIsSliding(), Translation="0,0,5", Rotation="0,0,0", Scale="1,1,1") -> (Pose) {
    alphaInputType=Bool
    rotationMode=BMM_Additive
    rotationSpace=BCS_BoneSpace
    translationMode=BMM_Additive
    translationSpace=BCS_BoneSpace
    alphaBoolBlend {blendInTime=0.1, blendOption=QuarticInOut, blendOutTime=0.4}
    boneToModify {boneName=dyn_jacket_r_01}
    translation {z=5}
   }
   n98 = "Transform (Modify) Bone - Bone: dyn_jacket_bk_02"(ComponentPose=@n97, bAlphaBoolEnabled=True, Translation="0,0,2", Rotation="0,0,0", Scale="1,1,1") {
    alphaInputType=Bool
    rotationMode=BMM_Additive
    rotationSpace=BCS_BoneSpace
    translationMode=BMM_Additive
    translationSpace=BCS_BoneSpace
    alphaBoolBlend {blendInTime=0.1, blendOption=QuarticInOut, blendOutTime=0.4}
    boneToModify {boneName=dyn_jacket_bk_02}
    translation {z=2}
   }
   Output Pose(Result=Component To Local(ComponentPose=@n98)) {layerGroup=None}
   n99 = Control Rig(Source=Copy Pose From Mesh() {bUseAttachedParent=true}) {
    alphaBoolBlend {blendOption=Linear}
    controlRigAssetReference {blueprintRigClass=&/M_MED_Majesty_Deform_CtrlRig_twists.M_MED_Majesty_Deform_CtrlRig_twists_C}
   }
   n100 = Control Rig!(
    Source=@n99
    Alpha
    KneeRot_fr_l_114deg:Vector
    KneeTrans_fr_l_114deg:Vector="0.258,-3.591,0.009"
    KneeScale_fr_l_114deg:Vector="1,1,1"
    KneeScale_fr_l_130deg:Vector="1,1,1"
    KneeRot_fr_l_130deg:Vector="0,0,-51.2"
    KneeTrans_fr_l_130deg:Vector="0.258,-3.591,0.009"
   ) -> (Pose) {
    alphaBoolBlend {blendOption=Linear}
    controlRigAssetReference {blueprintRigClass=&/M_MED_Majesty_Deform_CtrlRig_sculpters.M_MED_Majesty_Deform_CtrlRig_sculpters_C}
   }
   n101 = Pose Driver - Source: upperarm_l!(SourcePose:Pose Link Structure=@n100) -> (Pose:Pose Link Structure) {
    is=title AnimGraphNode_PoseDriver
    poseAsset=/BRCosmetics/Characters/Player/Male/Medium/Bodies/M_MED_Majesty/Meshes/Poses/M_MED_Majesty_upperarm_l_anim_PoseAsset
    onlyDriveBones [{boneName=dyn_lion_head_l}, {boneName=dyn_lion_jaw_l}]
    poseTargets [
     {
      bApplyCustomCurve=false
      bIsHidden=false
      distanceMethod=DefaultMethod
      drivenName=base_l
      functionType=DefaultFunction
      targetScale=0.4931
      boneTransforms [{targetRotation=<rot -37.7574 0.1509 2.6875>, targetTranslation=<15.5111 0 0>}]
      customCurve {defaultValue=3.4028234663852886e+38, postInfinityExtrap=RCCE_Constant, preInfinityExtrap=RCCE_Constant, keys []}
      targetRotation=<rot 0 0 0>
     }
     {
      bApplyCustomCurve=false
      bIsHidden=false
      distanceMethod=DefaultMethod
      drivenName=fwd_l
      functionType=DefaultFunction
      targetScale=0.6019
      boneTransforms [{targetRotation=<rot -37.517 68.0559 -71.2357>, targetTranslation=<15.5111 0 0.0003>}]
      customCurve {defaultValue=3.4028234663852886e+38, postInfinityExtrap=RCCE_Constant, preInfinityExtrap=RCCE_Constant, keys []}
      targetRotation=<rot 0 0 0>
     }
     {
      bApplyCustomCurve=false
      bIsHidden=false
      distanceMethod=DefaultMethod
      drivenName=dwn_l
      functionType=DefaultFunction
      targetScale=0.6847
      boneTransforms [{targetRotation=<rot -77.6502 105.7817 -85.0582>, targetTranslation=<15.5119 0.0029 -0.005>}]
      customCurve {defaultValue=3.4028234663852886e+38, postInfinityExtrap=RCCE_Constant, preInfinityExtrap=RCCE_Constant, keys []}
      targetRotation=<rot 0 0 0>
     }
     {
      bApplyCustomCurve=false
      bIsHidden=false
      distanceMethod=DefaultMethod
      drivenName=up_l
      functionType=DefaultFunction
      targetScale=1
      boneTransforms [{targetRotation=<rot 9.0555 55.813 -100.1185>, targetTranslation=<15.5106 0.0002 -0.0006>}]
      customCurve {defaultValue=3.4028234663852886e+38, postInfinityExtrap=RCCE_Constant, preInfinityExtrap=RCCE_Constant, keys []}
      targetRotation=<rot 0 0 0>
     }
     {
      bApplyCustomCurve=false
      bIsHidden=false
      distanceMethod=DefaultMethod
      drivenName=pick_l
      functionType=DefaultFunction
      targetScale=0.4931
      boneTransforms [{targetRotation=<rot -51.5614 31.7954 9.1643>, targetTranslation=<15.5111 -0.0001 0.0004>}]
      customCurve {defaultValue=3.4028234663852886e+38, postInfinityExtrap=RCCE_Constant, preInfinityExtrap=RCCE_Constant, keys []}
      targetRotation=<rot 0 0 0>
     }
    ]
    rBFParams {normalizeMethod=AlwaysNormalize, radius=50, solverType=Interpolative}
    sourceBones [{boneName=upperarm_l}]
   }
   n102 = Pose Driver - Source: upperarm_r!(SourcePose:Pose Link Structure=@n101) -> (Pose:Pose Link Structure) {
    is=title AnimGraphNode_PoseDriver
    poseAsset=/BRCosmetics/Characters/Player/Male/Medium/Bodies/M_MED_Majesty/Meshes/Poses/M_MED_Majesty_upperarm_r_anim_PoseAsset
    onlyDriveBones [{boneName=dyn_rion_head_r}, {boneName=dyn_rion_jaw_r}]
    poseTargets [
     {
      bApplyCustomCurve=false
      bIsHidden=false
      distanceMethod=DefaultMethod
      drivenName=base
      functionType=DefaultFunction
      targetScale=0.5767
      boneTransforms [{targetRotation=<rot -37.7574 0.1509 2.6875>, targetTranslation=<-15.5111 0 0>}]
      customCurve {defaultValue=3.4028234663852886e+38, postInfinityExtrap=RCCE_Constant, preInfinityExtrap=RCCE_Constant, keys []}
      targetRotation=<rot 0 0 0>
     }
     {
      bApplyCustomCurve=false
      bIsHidden=false
      distanceMethod=DefaultMethod
      drivenName=fwd
      functionType=DefaultFunction
      targetScale=0.9571
      boneTransforms [{targetRotation=<rot -8.8332 105.2654 -56.6207>, targetTranslation=<-15.511 0 0.0002>}]
      customCurve {defaultValue=3.4028234663852886e+38, postInfinityExtrap=RCCE_Constant, preInfinityExtrap=RCCE_Constant, keys []}
      targetRotation=<rot 0 0 0>
     }
     {
      bApplyCustomCurve=false
      bIsHidden=false
      distanceMethod=DefaultMethod
      drivenName=dwn
      functionType=DefaultFunction
      targetScale=0.7691
      boneTransforms [{targetRotation=<rot -48.8496 109.4839 -47.3731>, targetTranslation=<-15.5115 0.0019 -0.0027>}]
      customCurve {defaultValue=3.4028234663852886e+38, postInfinityExtrap=RCCE_Constant, preInfinityExtrap=RCCE_Constant, keys []}
      targetRotation=<rot 0 0 0>
     }
     {
      bApplyCustomCurve=false
      bIsHidden=false
      distanceMethod=DefaultMethod
      drivenName=up
      functionType=DefaultFunction
      targetScale=1
      boneTransforms [{targetRotation=<rot 34.603 97.4547 -83.2175>, targetTranslation=<-15.511 -0.0001 -0.0001>}]
      customCurve {defaultValue=3.4028234663852886e+38, postInfinityExtrap=RCCE_Constant, preInfinityExtrap=RCCE_Constant, keys []}
      targetRotation=<rot 0 0 0>
     }
     {
      bApplyCustomCurve=false
      bIsHidden=false
      distanceMethod=DefaultMethod
      drivenName=pick
      functionType=DefaultFunction
      targetScale=0.5767
      boneTransforms [{targetRotation=<rot -74.4373 27.0943 -26.0341>, targetTranslation=<-15.511 0 0>}]
      customCurve {defaultValue=3.4028234663852886e+38, postInfinityExtrap=RCCE_Constant, preInfinityExtrap=RCCE_Constant, keys []}
      targetRotation=<rot 0 0 0>
     }
    ]
    rBFParams {normalizeMethod=AlwaysNormalize, radius=50, solverType=Interpolative}
    sourceBones [{boneName=upperarm_r}]
   }
   n103 = Look At - Bone: dyn_shoe_tongue_l(ComponentPose=Local To Component(LocalPose=@n102), Alpha=1) {
    alphaBoolBlend {blendOption=Linear}
    boneToModify {boneName=dyn_shoe_tongue_l}
    lookAtLocation=<0 5>
    lookAtTarget {boneReference {boneName=calf_l}}
    lookAt_Axis {axis=<1 0>}
    lookUp_Axis {axis=<0 1>}
   }
   n104 = Look At - Bone: dyn_shoe_tongue_r(ComponentPose=@n103, Alpha=1) {
    alphaBoolBlend {blendOption=Linear}
    boneToModify {boneName=dyn_shoe_tongue_r}
    lookAtLocation=<0 -5>
    lookAtTarget {boneReference {boneName=calf_r}}
    lookAt_Axis {axis=<1 0>}
    lookUp_Axis {axis=<0 -1>}
   }
   n105 = Anim Dynamics - Bone: dyn_shoe_pull_l(ComponentPose=@n104, Alpha=1, ExternalForce="0,0,0") {
    angularDampingOverride=0.999
    angularSpringConstant=12000
    bAngularSpring=true
    bGravityOverrideInSimSpace=true
    bOverrideAngularDamping=true
    bUseGravityOverride=true
    gravityScale=0.35
    simulationSpace=RootRelative
    alphaBoolBlend {blendOption=Linear}
    boundBone {boneName=dyn_shoe_pull_l}
    componentLinearAccScale=<1 1 1>
    componentLinearVelScale=<1 1 1>
    gravityOverride {z=-980}
    physicsBodyDefinitions [{
      sphereCollisionRadius=0
      boundBone {boneName=dyn_shoe_pull_l}
      boxExtents=<20 10 10>
      constraintSetup {angularLimitsMax=<0 0 15>, angularLimitsMin=<0 0 -15>, angularTarget=<1 0 0>}
      localJointOffset=<4 0 0>
     }]
   }
   n106 = Anim Dynamics - Bone: dyn_shoe_pull_r(ComponentPose=@n105, Alpha=1, ExternalForce="0,0,0") {
    angularDampingOverride=0.999
    angularSpringConstant=12000
    bAngularSpring=true
    bGravityOverrideInSimSpace=true
    bOverrideAngularDamping=true
    bUseGravityOverride=true
    gravityScale=0.35
    simulationSpace=RootRelative
    alphaBoolBlend {blendOption=Linear}
    boundBone {boneName=dyn_shoe_pull_r}
    componentLinearAccScale=<1 1 1>
    componentLinearVelScale=<1 1 1>
    gravityOverride {z=-980}
    physicsBodyDefinitions [{
      sphereCollisionRadius=0
      boundBone {boneName=dyn_shoe_pull_r}
      boxExtents=<20 10 10>
      constraintSetup {angularLimitsMax=<0 0 15>, angularLimitsMin=<0 0 -15>, angularTarget=<1 0 0>}
      localJointOffset=<4 0 0>
     }]
   }
   n107 = Look At - Bone: spine_04_aimer(ComponentPose=@n106, Alpha=1) {
    alphaBoolBlend {blendOption=Linear}
    boneToModify {boneName=spine_04_aimer}
    lookAtLocation {x=0}
    lookAtTarget {boneReference {boneName=pelvis}}
    lookAt_Axis {axis=<-1 0>}
   }
   n108 = RigidBody(ComponentPose=@n107, Alpha=0.8, OverrideWorldGravity="0,0,-980") {
    bClampLinearTranslationLimitToRefPose=true
    bOverrideWorldGravity=true
    overridePhysicsAsset=&/M_MED_Majesty_Dyn_hood
    simulationSpace=BaseBoneSpace
    alphaBoolBlend {blendOption=Linear}
    baseBoneRef {boneName=spine_03}
    componentAppliedLinearAccClamp=<3000 3000 3000>
    componentLinearAccScale {z=0.5}
    componentLinearVelScale {z=0.5}
    overrideWorldGravity {z=-980}
    simSpaceSettings {
     maxAngularAcceleration=500
     maxAngularVelocity=3
     maxLinearAcceleration=700
     maxLinearVelocity=500
     velocityScaleZ=0
     worldAlpha=1
     externalLinearDragV=<1 1>
    }
   }
   n109 = Get bIsSliding()
   n110 = "Transform (Modify) Bone - Bone: dyn_jacket_fr_r_01_MSTR"(ComponentPose=@n108, bAlphaBoolEnabled=@n109, Translation="0,0,0", Rotation="0.000,20.000,90.000", Scale="1,1,1") {
    alphaInputType=Bool
    rotationMode=BMM_Additive
    rotationSpace=BCS_BoneSpace
    translationSpace=BCS_BoneSpace
    alphaBoolBlend {blendOption=Linear}
    alphaScaleBiasClamp {interpSpeedDecreasing=5, interpSpeedIncreasing=150}
    boneToModify {boneName=dyn_jacket_fr_r_01_MSTR}
    rotation=<rot 0 20 90>
   }
   n111 = "Transform (Modify) Bone - Bone: dyn_jacket_r_01_MSTR"(ComponentPose=@n110, bAlphaBoolEnabled=@n109, Translation="0,0,0", Rotation="10.000,30.000,90", Scale="1,1,1") {
    alphaInputType=Bool
    rotationMode=BMM_Additive
    rotationSpace=BCS_BoneSpace
    translationSpace=BCS_BoneSpace
    alphaBoolBlend {blendOption=Linear}
    alphaScaleBiasClamp {interpSpeedDecreasing=5, interpSpeedIncreasing=150}
    boneToModify {boneName=dyn_jacket_r_01_MSTR}
    rotation=<rot 10 30 90>
   }
   Save cached pose 'PreSim'(Pose=Component To Local(ComponentPose=@n111)) {cacheName=PreSim}
   Get bIsSliding()
  ]}
 EventGraph {nodes [
   n1 = Cast To PlayerPawn_Athena(Object:Wildcard)
   > PawnStartFire!(execute, self:Fort Pawn Object Reference=@n1, FireModeNum:Byte=0) -> (then)
   > Bind Event to On Group Emote Secondary Fire Pressed(self:Fort Player Pawn Object Reference=@n1, Delegate=OnGroupEmoteSecondaryFirePressed_Event() {is=title K2Node_CustomEvent}) {is=title K2Node_AddDelegate}
   Event BlueprintUpdateAnimation()
   > Is Valid(InputObject=Get SkeletalMesh() -> (SkeletalMesh:Skeletal Mesh Component Object Reference))
   n2 = Get SkeletalMesh()
   n3 = BreakTransform(InTransform=GetDeltaTransformFromRefPose(self=@n2, BoneName=clavicle_l, BaseName=pelvis))
   > Set arm rotation_l(arm rotation_l:double=MapRangeClamped(Value=BreakRotator(InRot=@n3.Rotation).Pitch, InRangeA=20, InRangeB=50)) -> (Output_Get:double)
   > Set leg_l_rotation fwd(leg_l_rotation fwd:double=GetAlphaFromReferenceTransformDelta(SkeletalMeshComponent=@n2, BoneName=thigh_l, BaseName=pelvis, InputRange=(X=-20,Y=-100), TwistAxis="1,0,0")) -> (Output_Get:double)
   > Set leg_r_rotation_fwd(leg_r_rotation_fwd:double=GetAlphaFromReferenceTransformDelta(SkeletalMeshComponent=@n2, BoneName=thigh_r, BaseName=pelvis, InputRange=(X=-20,Y=-100), TwistAxis="1,0,0")) -> (Output_Get:double)
   > Set Head Rotation(Head Rotation:double=GetAlphaFromReferenceTransformDelta(SkeletalMeshComponent=@n2, BoneName=head, BaseName=pelvis, InputRange=(X=-24,Y=-55), TwistAxis="1,0,0")) -> (Output_Get:double)
   n4 = Vector Length(A=Get PawnVelocity())
   n5 = Get bIsJumping()
   n6 = "To Float (Boolean)"(InBool=AND Boolean(A=Get bIsMoving2D(), B=NOT Boolean(A=@n5)))
   > Set Alpha_jacket(Alpha_jacket:double=MapRangeClamped(Value=float * float(A=@n4, B=@n6) {is=title K2Node_PromotableOperator}, InRangeA=500, InRangeB=770, OutRangeB=0.2)) -> (Output_Get:double)
   n7 = Get SkeletalMesh()
   > Set Thigh_l(Thigh_l:float=GetAlphaFromReferenceTransformDelta(SkeletalMeshComponent=@n7, BoneName=thigh_l, BaseName=pelvis, InputRange=(X=-10,Y=-90), TwistAxis="1,0,0")) -> (Output_Get:float)
   > Set Thigh_r(Thigh_r:float=GetAlphaFromReferenceTransformDelta(SkeletalMeshComponent=@n7, BoneName=thigh_r, BaseName=pelvis, InputRange=(X=-10,Y=-90), TwistAxis="1,0,0")) -> (Output_Get:float)
   n8 = Get bIsCrouching()
   n9 = Get bIsPlayingMeleeAnim()
   n10 = OR Boolean(A=Get bIsJumping(), B=Get bIsFalling())
   > Set jump_melee(jump_melee:bool=OR Boolean(A=AND Boolean(A=@n10, B=@n9), B=AND Boolean(A=@n9, B=@n8))) -> (Output_Get:bool)
   n11 = Get SkeletalMesh()
   n12 = Get bIsSliding()
   > Set sliding gun(sliding gun:double=float * float(
     A=GetAlphaFromReferenceTransformDelta(SkeletalMeshComponent=@n11, BoneName=spine_05, BaseName=pelvis, DeltaComponentType=EulerX, InputRange=(X=-20,Y=-45), TwistAxis="1,0,0")
     B="To Float (Boolean)"(InBool=@n12)
    )) -> (Output_Get:double)
   Event OnCharacterCustomizationLoaded()
   > Set SkeletalMesh(SkeletalMesh:Skeletal Mesh Component Object Reference=GetOwnerSkeletalMeshComponent()) -> (Output_Get:Skeletal Mesh Component Object Reference)
   NOT Boolean(A=Get bIsSliding())
   GetAlphaFromReferenceTransformDelta(BoneName=thigh_r, BaseName=pelvis, InputRange=(X=0,Y=-90), OutputRange=(X=0,Y=0.7), TwistAxis="1,0,1")
  ]}
]
variables [
 Alpha_jacket {type=number}
 arm rotation_l {type=number}
 Head Rotation {type=number}
 jump_melee {type=boolean}
 leg_l_rotation fwd {type=number}
 leg_r_rotation_fwd {type=number}
 SkeletalMesh {type=SkeletalMeshComponent}
 sliding gun {type=number}
 Stamina Weapon {type=number}
 Thigh_l {type=number}
 Thigh_r {type=number}
]
