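# Animation Blueprint for the Majesty medium male body cosmetic part, adding procedural motion to its dyn_crown bone.
# The AnimGraph copies the attached parent's pose, then offsets the crown with an additive bone-space translation.
# Each update, the EventGraph computes that offset from a game-time-driven random vector function with per-axis frequency and amplitude.
# A partially blended Anim Dynamics node then adds spring-damped secondary motion with reduced gravity, simulated root-relative.
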AnimBlueprint /BRCosmetics/Characters/Player/Male/Medium/Bodies/M_MED_Majesty/Meshes/Parts/M_MED_Majesty_AnimBP
target_skeleton=&/M_MED_Majesty_Skeleton
graphs [
 AnimGraph {nodes [
   n1 = "Transform (Modify) Bone - Bone: dyn_crown"(
    ComponentPose=Local To Component(LocalPose=Copy Pose From Mesh() {bUseAttachedParent=true})
    Alpha=1
    Translation=Get crown movement() -> (crown movement:Vector)
    Rotation="0,0,0"
    Scale="1,1,1"
   ) {
    rotationSpace=BCS_WorldSpace
    translationMode=BMM_Additive
    translationSpace=BCS_BoneSpace
    alphaBoolBlend {blendOption=Linear}
    alphaScaleBiasClamp {interpSpeedDecreasing=2, interpSpeedIncreasing=2}
    boneToModify {boneName=dyn_crown}
   }
   n2 = Anim Dynamics - Bone: dyn_crown(ComponentPose=@n1, Alpha=0.3, ExternalForce="0,0,0") {
    angularDampingOverride=1
    angularSpringConstant=20000
    bAngularSpring=true
    bGravityOverrideInSimSpace=true
    bLinearSpring=true
    bOverrideAngularDamping=true
    bOverrideLinearDamping=true
    bUseGravityOverride=true
    gravityScale=0.35
    linearDampingOverride=1
    linearSpringConstant=150
    simulationSpace=RootRelative
    alphaBoolBlend {blendOption=Linear}
    boundBone {boneName=dyn_crown}
    componentAppliedLinearAccClamp=<3000 3000 3000>
    componentLinearAccScale=<0.3 0.3 1>
    componentLinearVelScale=<0.3 0.3 1>
    gravityOverride {z=-980}
    physicsBodyDefinitions [{
      sphereCollisionRadius=0
      boundBone {boneName=dyn_crown}
      boxExtents=<20 30 30>
      constraintSetup {angularLimitsMax=<0 0 0>, angularLimitsMin=<0 0 0>, angularTarget=<1 0 0>, linearAxesMax=<0 50 50>, linearAxesMin=<-50 -50 -50>}
      localJointOffset=<0 0 0>
     }]
   }
   Output Pose(Result=Component To Local(ComponentPose=@n2)) {layerGroup=None}
  ]}
 EventGraph {nodes [
   Event BlueprintUpdateAnimation()
   > n1 = Random Vector Function!(
    execute
    CharacterPartAnimInstance:Custom Character Part Anim Instance Object Reference=Self-Reference()
    SkelMeshComponent:Mesh Component Object Reference
    JointName:Name=None
    GameTimeInSeconds:double=GetGameTimeInSeconds()
    TimeOffset:double=0
    PerlinScale:double=0
    Frequency:Vector="0.9,1.1,0.75"
    Amplitude:Vector="0.3,0.7,0.5"
    Offset:Vector="0,0,0"
   ) -> (then, VectorOut:Vector)
   > Set crown movement(crown movement:Vector=@n1) -> (Output_Get:Vector)
  ]}
]
variables [crown movement {type=Vector}]
